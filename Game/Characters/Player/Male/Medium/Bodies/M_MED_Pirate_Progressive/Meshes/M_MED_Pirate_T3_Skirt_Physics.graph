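# Physics asset for the tier 3 skirt of the medium male Pirate player body, setting up the skirt bones for physical motion around the legs.
# Eight five-bone chains of thin boxes (front, side, mid and back on each side) hang from the pelvis, with tight twist limits and narrower second swing at the back roots.
# Capsules cover pelvis, thighs, calves and only the left foot.
# Angularly free cross-links join each side's fourth mid and back segments.
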PhysicsAsset /Game/Characters/Player/Male/Medium/Bodies/M_MED_Pirate_Progressive/Meshes/M_MED_Pirate_T3_Skirt_Physics
bodies {
 calf_l [{shapeType=Capsule, length=32.75, radius=13, center=<-25 0 0>, rotation=<rot 90 0 0>}]
 calf_r [{shapeType=Capsule, length=34.75, radius=11.25, center=<25 0 0>, rotation=<rot 90 0 0>}]
 dyn_skirt_back_01_l [{shapeType=Box, extentX=15, extentY=2, extentZ=20, center=<8 0 0>}]
 dyn_skirt_back_01_r [{shapeType=Box, extentX=15, extentY=2, extentZ=20, center=<8 0 0>}]
 dyn_skirt_back_02_l [{shapeType=Box, extentX=15, extentY=2, extentZ=20, center=<8 0 0>}]
 dyn_skirt_back_02_r [{shapeType=Box, extentX=15, extentY=2, extentZ=20, center=<8 0 0>}]
 dyn_skirt_back_03_l [{shapeType=Box, extentX=15, extentY=2, extentZ=20, center=<8 0 0>}]
 dyn_skirt_back_03_r [{shapeType=Box, extentX=15, extentY=2, extentZ=20, center=<8 0 0>}]
 dyn_skirt_back_04_l [{shapeType=Box, extentX=15, extentY=2, extentZ=17.5, center=<8 0 0>}]
 dyn_skirt_back_04_r [{shapeType=Box, extentX=15, extentY=2, extentZ=17.5, center=<8 0 0>}]
 dyn_skirt_back_05_l [{shapeType=Box, extentX=15, extentY=2, extentZ=17.5, center=<8 0 0>}]
 dyn_skirt_back_05_r [{shapeType=Box, extentX=15, extentY=2, extentZ=17.5, center=<8 0 0>}]
 dyn_skirt_front_01_l [{shapeType=Box, extentX=15, extentY=2, extentZ=20, center=<8 0 0>}]
 dyn_skirt_front_01_r [{shapeType=Box, extentX=15, extentY=2, extentZ=20, center=<8 0 0>}]
 dyn_skirt_front_02_l [{shapeType=Box, extentX=15, extentY=2, extentZ=20, center=<8 0 0>, rotation=<rot 0 0 15.6869>}]
 dyn_skirt_front_02_r [{shapeType=Box, extentX=15, extentY=2, extentZ=20, center=<8 0 0>}]
 dyn_skirt_front_03_l [{shapeType=Box, extentX=15, extentY=2, extentZ=20, center=<8 0 0>, rotation=<rot 0 -0.0001 20.0592>}]
 dyn_skirt_front_03_r [{shapeType=Box, extentX=15, extentY=2, extentZ=20, center=<8 0 0>}]
 dyn_skirt_front_04_l [{shapeType=Box, extentX=15, extentY=2, extentZ=20, center=<8 1.0526 1.997>, rotation=<rot 0 0 24.7817>}]
 dyn_skirt_front_04_r [{shapeType=Box, extentX=15, extentY=2, extentZ=20, center=<8 0 0>}]
 dyn_skirt_front_05_l [{shapeType=Box, extentX=15, extentY=2, extentZ=20, center=<8 1.417 3.3209>, rotation=<rot 0 0 23.1077>}]
 dyn_skirt_front_05_r [{shapeType=Box, extentX=15, extentY=2, extentZ=20, center=<8 0 0>}]
 dyn_skirt_mid_01_l [{shapeType=Box, extentX=15, extentY=2, extentZ=20, center=<8 0 0>}]
 dyn_skirt_mid_01_r [{shapeType=Box, extentX=15, extentY=2, extentZ=20, center=<8 0 0>}]
 dyn_skirt_mid_02_l [{shapeType=Box, extentX=15, extentY=2, extentZ=20, center=<8 0 0>}]
 dyn_skirt_mid_02_r [{shapeType=Box, extentX=15, extentY=2, extentZ=20, center=<8 0 0>}]
 dyn_skirt_mid_03_l [{shapeType=Box, extentX=15, extentY=2, extentZ=20, center=<8 0 0>}]
 dyn_skirt_mid_03_r [{shapeType=Box, extentX=15, extentY=2, extentZ=20, center=<8 0 0>}]
 dyn_skirt_mid_04_l [{shapeType=Box, extentX=15, extentY=2, extentZ=20, center=<8 0 0>}]
 dyn_skirt_mid_04_r [{shapeType=Box, extentX=15, extentY=2, extentZ=20, center=<8 0 0>}]
 dyn_skirt_mid_05_l [{shapeType=Box, extentX=15, extentY=2, extentZ=20, center=<8 0 0>}]
 dyn_skirt_mid_05_r [{shapeType=Box, extentX=15, extentY=2, extentZ=20, center=<8 0 0>}]
 dyn_skirt_side_01_l [{shapeType=Box, extentX=15, extentY=2, extentZ=20, center=<8 0 0>}]
 dyn_skirt_side_01_r [{shapeType=Box, extentX=15, extentY=2, extentZ=20, center=<8 0 0>}]
 dyn_skirt_side_02_l [{shapeType=Box, extentX=15, extentY=2, extentZ=20, center=<8 0 0>}]
 dyn_skirt_side_02_r [{shapeType=Box, extentX=15, extentY=2, extentZ=20, center=<8 0 0>}]
 dyn_skirt_side_03_l [{shapeType=Box, extentX=15, extentY=2, extentZ=20, center=<8 0 0>}]
 dyn_skirt_side_03_r [{shapeType=Box, extentX=15, extentY=2, extentZ=20, center=<8 0 0>}]
 dyn_skirt_side_04_l [{shapeType=Box, extentX=15, extentY=2, extentZ=20, center=<8 0 0>}]
 dyn_skirt_side_04_r [{shapeType=Box, extentX=15, extentY=2, extentZ=20, center=<8 0 0>}]
 dyn_skirt_side_05_l [{shapeType=Box, extentX=15, extentY=2, extentZ=20, center=<8 0 0>}]
 dyn_skirt_side_05_r [{shapeType=Box, extentX=15, extentY=2, extentZ=20, center=<8 0 0>}]
 foot_l [{shapeType=Capsule, length=17.5223, radius=7.1179, center=<-3.6475 7.3109 0>, rotation=<rot 0 0 90>}]
 pelvis [{shapeType=Capsule, length=11.174, radius=9.5394, center=<-8.2664 -0.0002 0>}]
 thigh_l [
  {shapeType=Capsule, length=25, radius=9, center=<-25 0 0>, rotation=<rot 90 0 0>}
  {shapeType=Capsule, length=17, radius=9, center=<-25.041 4.4269 0>, rotation=<rot 90 0 0>}
 ]
 thigh_r [
  {shapeType=Capsule, length=25, radius=9, center=<25 0 0>, rotation=<rot 90 0 180>}
  {shapeType=Capsule, length=17, radius=9, center=<25.041 -4.4269 0>, rotation=<rot 90 0 180>}
 ]
}
constraints [
 {
  bone1Name=calf_l
  bone2Name=thigh_l
  swing1LimitDegrees=45
  swing1Motion=Free
  swing2LimitDegrees=45
  swing2Motion=Free
  twistLimitDegrees=45
  twistMotion=Free
 }
 {
  bone1Name=foot_l
  bone2Name=calf_l
  swing1LimitDegrees=45
  swing1Motion=Free
  swing2LimitDegrees=45
  swing2Motion=Free
  twistLimitDegrees=45
  twistMotion=Free
 }
 {
  bone1Name=thigh_l
  bone2Name=pelvis
  swing1LimitDegrees=45
  swing1Motion=Free
  swing2LimitDegrees=45
  swing2Motion=Free
  twistLimitDegrees=45
  twistMotion=Free
 }
 {
  bone1Name=calf_r
  bone2Name=thigh_r
  swing1LimitDegrees=45
  swing1Motion=Free
  swing2LimitDegrees=45
  swing2Motion=Free
  twistLimitDegrees=45
  twistMotion=Free
 }
 {bone1Name=dyn_skirt_back_04_l, bone2Name=dyn_skirt_back_03_l, swing1LimitDegrees=45, swing2LimitDegrees=45, twistLimitDegrees=5}
 {bone1Name=dyn_skirt_front_01_r, bone2Name=pelvis, swing1LimitDegrees=45, swing2LimitDegrees=45, twistLimitDegrees=5}
 {bone1Name=dyn_skirt_front_02_r, bone2Name=dyn_skirt_front_01_r, swing1LimitDegrees=45, swing2LimitDegrees=45, twistLimitDegrees=5}
 {bone1Name=dyn_skirt_front_03_r, bone2Name=dyn_skirt_front_02_r, swing1LimitDegrees=45, swing2LimitDegrees=45, twistLimitDegrees=5}
 {bone1Name=dyn_skirt_front_04_r, bone2Name=dyn_skirt_front_03_r, swing1LimitDegrees=45, swing2LimitDegrees=45, twistLimitDegrees=5}
 {bone1Name=dyn_skirt_side_01_r, bone2Name=pelvis, swing1LimitDegrees=45, swing2LimitDegrees=45, twistLimitDegrees=5}
 {bone1Name=dyn_skirt_side_02_r, bone2Name=dyn_skirt_side_01_r, swing1LimitDegrees=45, swing2LimitDegrees=45, twistLimitDegrees=5}
 {bone1Name=dyn_skirt_side_03_r, bone2Name=dyn_skirt_side_02_r, swing1LimitDegrees=45, swing2LimitDegrees=45, twistLimitDegrees=5}
 {bone1Name=dyn_skirt_side_04_r, bone2Name=dyn_skirt_side_03_r, swing1LimitDegrees=45, swing2LimitDegrees=45, twistLimitDegrees=5}
 {bone1Name=dyn_skirt_mid_01_r, bone2Name=pelvis, swing1LimitDegrees=45, swing2LimitDegrees=45, twistLimitDegrees=5}
 {bone1Name=dyn_skirt_mid_02_r, bone2Name=dyn_skirt_mid_01_r, swing1LimitDegrees=45, swing2LimitDegrees=45, twistLimitDegrees=5}
 {bone1Name=dyn_skirt_mid_03_r, bone2Name=dyn_skirt_mid_02_r, swing1LimitDegrees=45, swing2LimitDegrees=45, twistLimitDegrees=5}
 {bone1Name=dyn_skirt_mid_04_r, bone2Name=dyn_skirt_mid_03_r, swing1LimitDegrees=45, swing2LimitDegrees=45, twistLimitDegrees=5}
 {bone1Name=dyn_skirt_back_01_r, bone2Name=pelvis, swing1LimitDegrees=45, swing2LimitDegrees=1, twistLimitDegrees=5}
 {bone1Name=dyn_skirt_back_02_r, bone2Name=dyn_skirt_back_01_r, swing1LimitDegrees=45, swing2LimitDegrees=45, twistLimitDegrees=5}
 {bone1Name=dyn_skirt_back_03_r, bone2Name=dyn_skirt_back_02_r, swing1LimitDegrees=45, swing2LimitDegrees=45, twistLimitDegrees=5}
 {bone1Name=dyn_skirt_back_04_r, bone2Name=dyn_skirt_back_03_r, swing1LimitDegrees=45, swing2LimitDegrees=45, twistLimitDegrees=5}
 {bone1Name=dyn_skirt_front_01_l, bone2Name=pelvis, swing1LimitDegrees=45, swing2LimitDegrees=45, twistLimitDegrees=5}
 {bone1Name=dyn_skirt_front_02_l, bone2Name=dyn_skirt_front_01_l, swing1LimitDegrees=45, swing2LimitDegrees=45, twistLimitDegrees=5}
 {bone1Name=dyn_skirt_front_03_l, bone2Name=dyn_skirt_front_02_l, swing1LimitDegrees=45, swing2LimitDegrees=45, twistLimitDegrees=5}
 {bone1Name=dyn_skirt_front_04_l, bone2Name=dyn_skirt_front_03_l, swing1LimitDegrees=45, swing2LimitDegrees=45, twistLimitDegrees=5}
 {bone1Name=dyn_skirt_side_01_l, bone2Name=pelvis, swing1LimitDegrees=45, swing2LimitDegrees=45, twistLimitDegrees=5}
 {bone1Name=dyn_skirt_side_02_l, bone2Name=dyn_skirt_side_01_l, swing1LimitDegrees=45, swing2LimitDegrees=45, twistLimitDegrees=5}
 {bone1Name=dyn_skirt_side_03_l, bone2Name=dyn_skirt_side_02_l, swing1LimitDegrees=45, swing2LimitDegrees=45, twistLimitDegrees=5}
 {bone1Name=dyn_skirt_side_04_l, bone2Name=dyn_skirt_side_03_l, swing1LimitDegrees=45, swing2LimitDegrees=45, twistLimitDegrees=5}
 {bone1Name=dyn_skirt_mid_01_l, bone2Name=pelvis, swing1LimitDegrees=45, swing2LimitDegrees=45, twistLimitDegrees=5}
 {bone1Name=dyn_skirt_mid_02_l, bone2Name=dyn_skirt_mid_01_l, swing1LimitDegrees=45, swing2LimitDegrees=45, twistLimitDegrees=5}
 {bone1Name=dyn_skirt_mid_03_l, bone2Name=dyn_skirt_mid_02_l, swing1LimitDegrees=45, swing2LimitDegrees=45, twistLimitDegrees=5}
 {bone1Name=dyn_skirt_mid_04_l, bone2Name=dyn_skirt_mid_03_l, swing1LimitDegrees=45, swing2LimitDegrees=45, twistLimitDegrees=5}
 {bone1Name=dyn_skirt_back_01_l, bone2Name=pelvis, swing1LimitDegrees=45, swing2LimitDegrees=1, twistLimitDegrees=5}
 {bone1Name=dyn_skirt_back_02_l, bone2Name=dyn_skirt_back_01_l, swing1LimitDegrees=45, swing2LimitDegrees=45, twistLimitDegrees=5}
 {bone1Name=dyn_skirt_back_03_l, bone2Name=dyn_skirt_back_02_l, swing1LimitDegrees=45, swing2LimitDegrees=45, twistLimitDegrees=5}
 {
  bone1Name=thigh_r
  bone2Name=pelvis
  swing1LimitDegrees=45
  swing1Motion=Free
  swing2LimitDegrees=45
  swing2Motion=Free
  twistLimitDegrees=45
  twistMotion=Free
 }
 {bone1Name=dyn_skirt_front_05_r, bone2Name=dyn_skirt_front_04_r, swing1LimitDegrees=45, swing2LimitDegrees=45, twistLimitDegrees=5}
 {bone1Name=dyn_skirt_side_05_r, bone2Name=dyn_skirt_side_04_r, swing1LimitDegrees=45, swing2LimitDegrees=45, twistLimitDegrees=5}
 {bone1Name=dyn_skirt_mid_05_r, bone2Name=dyn_skirt_mid_04_r, swing1LimitDegrees=45, swing2LimitDegrees=45, twistLimitDegrees=5}
 {bone1Name=dyn_skirt_back_05_r, bone2Name=dyn_skirt_back_04_r, swing1LimitDegrees=45, swing2LimitDegrees=45, twistLimitDegrees=5}
 {bone1Name=dyn_skirt_front_05_l, bone2Name=dyn_skirt_front_04_l, swing1LimitDegrees=45, swing2LimitDegrees=45, twistLimitDegrees=5}
 {bone1Name=dyn_skirt_side_05_l, bone2Name=dyn_skirt_side_04_l, swing1LimitDegrees=45, swing2LimitDegrees=45, twistLimitDegrees=5}
 {bone1Name=dyn_skirt_mid_05_l, bone2Name=dyn_skirt_mid_04_l, swing1LimitDegrees=45, swing2LimitDegrees=45, twistLimitDegrees=5}
 {bone1Name=dyn_skirt_back_05_l, bone2Name=dyn_skirt_back_04_l, swing1LimitDegrees=45, swing2LimitDegrees=45, twistLimitDegrees=5}
 {
  bone1Name=dyn_skirt_mid_04_l
  bone2Name=dyn_skirt_back_04_l
  swing1LimitDegrees=45
  swing1Motion=Free
  swing2LimitDegrees=45
  swing2Motion=Free
  twistLimitDegrees=45
  twistMotion=Free
 }
 {
  bone1Name=dyn_skirt_mid_04_r
  bone2Name=dyn_skirt_back_04_r
  swing1LimitDegrees=45
  swing1Motion=Free
  swing2LimitDegrees=45
  swing2Motion=Free
  twistLimitDegrees=45
  twistMotion=Free
 }
]
