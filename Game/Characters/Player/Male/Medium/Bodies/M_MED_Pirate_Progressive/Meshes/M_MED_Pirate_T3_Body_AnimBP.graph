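# Anim Blueprint for the tier 3 medium male pirate body, adding procedural motion to shoulder pads, tassels, peg-leg cloth, hook and necklace.
# RBF Pose Drivers pose shoulder pads from upper-arm and clavicle rotation.
# A bool blend selects the no-rigid-body cached pose when rigid bodies are disabled or LOD exceeds 1.
# Weighted constraints then anchor tassels before AnimDynamics simulation, with skydiving force on main tassels; a Look At aims the hook.
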AnimBlueprint /Game/Characters/Player/Male/Medium/Bodies/M_MED_Pirate_Progressive/Meshes/M_MED_Pirate_T3_Body_AnimBP
target_skeleton=&/M_MED_Pirate_T3_Body_Skeleton
graphs [
 AnimGraph {nodes [
   n1 = Pose Driver - Source: upperarm_l!(SourcePose:Pose Link Structure) -> (Pose:Pose Link Structure) {
    is=title AnimGraphNode_PoseDriver
    poseAsset=&/M_MED_Pirate_Body_Shoulder_Skydiving_RBF_PoseAsset
    onlyDriveBones [{boneName=dyn_shoulderpad_l}, {boneName=pd_shoulderPad_l}, {boneName=dyn_shoulderPad_fr_l}]
    poseTargets [
     {
      bApplyCustomCurve=false
      bIsHidden=false
      distanceMethod=DefaultMethod
      drivenName=Pose_12
      functionType=DefaultFunction
      targetScale=0.6185
      boneTransforms [
       {targetRotation=<rot -64.7854 -3.089 6.0959>, targetTranslation=<15.5111 0 0>}
       {targetRotation=<rot -85.1915 -138.5625 143.0619>, targetTranslation=<4.3575 -0.1466 -2.3223>}
      ]
      customCurve {defaultValue=3.4028234663852886e+38, postInfinityExtrap=RCCE_Constant, preInfinityExtrap=RCCE_Constant, keys []}
      targetRotation=<rot 0 0 0>
     }
     {
      bApplyCustomCurve=false
      bIsHidden=false
      distanceMethod=DefaultMethod
      drivenName=Pose_13
      functionType=DefaultFunction
      targetScale=0.8674
      boneTransforms [
       {targetRotation=<rot -26.5845 -8.253 -68.9274>, targetTranslation=<15.5111 0 0>}
       {targetRotation=<rot -64.8583 -92.9709 99.1838>, targetTranslation=<4.3575 -0.1466 -2.3223>}
      ]
      customCurve {defaultValue=3.4028234663852886e+38, postInfinityExtrap=RCCE_Constant, preInfinityExtrap=RCCE_Constant, keys []}
      targetRotation=<rot 0 0 0>
     }
     {
      bApplyCustomCurve=false
      bIsHidden=false
      distanceMethod=DefaultMethod
      drivenName=Pose_14
      functionType=DefaultFunction
      targetScale=0.3953
      boneTransforms [
       {targetRotation=<rot 18.9792 7.7335 -91.3654>, targetTranslation=<15.5111 0 0>}
       {targetRotation=<rot -64.8617 -92.9723 99.185>, targetTranslation=<4.3575 -0.1466 -2.3223>}
      ]
      customCurve {defaultValue=3.4028234663852886e+38, postInfinityExtrap=RCCE_Constant, preInfinityExtrap=RCCE_Constant, keys []}
      targetRotation=<rot 0 0 0>
     }
     {
      bApplyCustomCurve=false
      bIsHidden=false
      distanceMethod=DefaultMethod
      drivenName=Pose_15
      functionType=DefaultFunction
      targetScale=0.3953
      boneTransforms [
       {targetRotation=<rot 15.4337 14.8613 -68.8945>, targetTranslation=<15.5111 0 0>}
       {targetRotation=<rot -64.8582 -92.9709 99.1838>, targetTranslation=<4.3575 -0.1466 -2.3223>}
      ]
      customCurve {defaultValue=3.4028234663852886e+38, postInfinityExtrap=RCCE_Constant, preInfinityExtrap=RCCE_Constant, keys []}
      targetRotation=<rot 0 0 0>
     }
     {
      bApplyCustomCurve=false
      bIsHidden=false
      distanceMethod=DefaultMethod
      drivenName=Pose_16
      functionType=DefaultFunction
      targetScale=1
      boneTransforms [
       {targetRotation=<rot -24.0458 -41.0619 -7.7393>, targetTranslation=<15.5111 0 0>}
       {targetRotation=<rot -64.8025 -92.9475 99.1643>, targetTranslation=<4.3575 -0.1466 -2.3223>}
      ]
      customCurve {defaultValue=3.4028234663852886e+38, postInfinityExtrap=RCCE_Constant, preInfinityExtrap=RCCE_Constant, keys []}
      targetRotation=<rot 0 0 0>
     }
     {
      bApplyCustomCurve=false
      bIsHidden=false
      distanceMethod=DefaultMethod
      drivenName=Pose_17
      functionType=DefaultFunction
      targetScale=0.6185
      boneTransforms [
       {targetRotation=<rot -46.7745 -56.0882 44.8776>, targetTranslation=<15.5111 0 0>}
       {targetRotation=<rot -85.1956 -138.5362 143.0321>, targetTranslation=<4.3575 -0.1466 -2.3223>}
      ]
      customCurve {defaultValue=3.4028234663852886e+38, postInfinityExtrap=RCCE_Constant, preInfinityExtrap=RCCE_Constant, keys []}
      targetRotation=<rot 0 0 0>
     }
    ]
    rBFParams {distanceMethod=Quaternion, function=Linear, radius=50}
    sourceBones [{boneName=upperarm_l}, {boneName=clavicle_l}]
   }
   Pose Driver - Source: upperarm_r!(SourcePose:Pose Link Structure=@n1) -> (Pose:Pose Link Structure) {
    is=title AnimGraphNode_PoseDriver
    poseAsset=&M_MED_Pirate_Body_Shoulder_Skydiving_RBF_PoseAsset
    onlyDriveBones [{boneName=dyn_shoulderpad_r}, {boneName=pd_shoulderPad_r}, {boneName=dyn_shoulderPad_fr_r}]
    poseTargets [
     {
      bApplyCustomCurve=false
      bIsHidden=false
      distanceMethod=DefaultMethod
      drivenName=Pose_12
      functionType=DefaultFunction
      targetScale=0.6058
      boneTransforms [
       {targetRotation=<rot -64.7854 -3.089 6.0959>, targetTranslation=<-15.5111 0 0>}
       {targetRotation=<rot -85.1915 41.4375 143.0619>, targetTranslation=<4.3575 -0.1466 2.3223>}
      ]
      customCurve {
       defaultValue=3.4028234663852886e+38
       postInfinityExtrap=RCCE_Constant
       preInfinityExtrap=RCCE_Constant
       keys [
        {
         arriveTangent=0
         arriveTangentWeight=0
         interpMode=RCIM_Linear
         leaveTangent=0
         leaveTangentWeight=0
         tangentMode=RCTM_Auto
         tangentWeightMode=RCTWM_WeightedNone
         time=0
         value=0
        }
        {
         arriveTangent=0
         arriveTangentWeight=0
         interpMode=RCIM_Linear
         leaveTangent=0
         leaveTangentWeight=0
         tangentMode=RCTM_Auto
         tangentWeightMode=RCTWM_WeightedNone
         time=1
         value=1
        }
       ]
      }
      targetRotation=<rot 0 0 0>
     }
     {
      bApplyCustomCurve=false
      bIsHidden=false
      distanceMethod=DefaultMethod
      drivenName=Pose_13
      functionType=DefaultFunction
      targetScale=0.2437
      boneTransforms [
       {targetRotation=<rot -4.2279 0.0137 -58.9422>, targetTranslation=<-15.5111 0 0>}
       {targetRotation=<rot -64.5761 87.1291 99.0963>, targetTranslation=<4.3575 -0.1466 2.3223>}
      ]
      customCurve {
       defaultValue=3.4028234663852886e+38
       postInfinityExtrap=RCCE_Constant
       preInfinityExtrap=RCCE_Constant
       keys [
        {
         arriveTangent=0
         arriveTangentWeight=0
         interpMode=RCIM_Linear
         leaveTangent=0
         leaveTangentWeight=0
         tangentMode=RCTM_Auto
         tangentWeightMode=RCTWM_WeightedNone
         time=0
         value=0
        }
        {
         arriveTangent=0
         arriveTangentWeight=0
         interpMode=RCIM_Linear
         leaveTangent=0
         leaveTangentWeight=0
         tangentMode=RCTM_Auto
         tangentWeightMode=RCTWM_WeightedNone
         time=1
         value=1
        }
       ]
      }
      targetRotation=<rot 0 0 0>
     }
     {
      bApplyCustomCurve=false
      bIsHidden=false
      distanceMethod=DefaultMethod
      drivenName=Pose_14
      functionType=DefaultFunction
      targetScale=1
      boneTransforms [
       {targetRotation=<rot 24.6296 27.9377 -99.4979>, targetTranslation=<-15.5111 0 0>}
       {targetRotation=<rot -64.5798 87.1276 99.0977>, targetTranslation=<4.3575 -0.1466 2.3223>}
      ]
      customCurve {
       defaultValue=3.4028234663852886e+38
       postInfinityExtrap=RCCE_Constant
       preInfinityExtrap=RCCE_Constant
       keys [
        {
         arriveTangent=0
         arriveTangentWeight=0
         interpMode=RCIM_Linear
         leaveTangent=0
         leaveTangentWeight=0
         tangentMode=RCTM_Auto
         tangentWeightMode=RCTWM_WeightedNone
         time=0
         value=0
        }
        {
         arriveTangent=0
         arriveTangentWeight=0
         interpMode=RCIM_Linear
         leaveTangent=0
         leaveTangentWeight=0
         tangentMode=RCTM_Auto
         tangentWeightMode=RCTWM_WeightedNone
         time=1
         value=1
        }
       ]
      }
      targetRotation=<rot 0 0 0>
     }
     {
      bApplyCustomCurve=false
      bIsHidden=false
      distanceMethod=DefaultMethod
      drivenName=Pose_15
      functionType=DefaultFunction
      targetScale=0.2437
      boneTransforms [
       {targetRotation=<rot -16.6572 -7.7477 -57.8379>, targetTranslation=<-15.5111 0 0>}
       {targetRotation=<rot -64.5765 87.1288 99.0967>, targetTranslation=<4.3575 -0.1466 2.3223>}
      ]
      customCurve {
       defaultValue=3.4028234663852886e+38
       postInfinityExtrap=RCCE_Constant
       preInfinityExtrap=RCCE_Constant
       keys [
        {
         arriveTangent=0
         arriveTangentWeight=0
         interpMode=RCIM_Linear
         leaveTangent=0
         leaveTangentWeight=0
         tangentMode=RCTM_Auto
         tangentWeightMode=RCTWM_WeightedNone
         time=0
         value=0
        }
        {
         arriveTangent=0
         arriveTangentWeight=0
         interpMode=RCIM_Linear
         leaveTangent=0
         leaveTangentWeight=0
         tangentMode=RCTM_Auto
         tangentWeightMode=RCTWM_WeightedNone
         time=1
         value=1
        }
       ]
      }
      targetRotation=<rot 0 0 0>
     }
     {
      bApplyCustomCurve=false
      bIsHidden=false
      distanceMethod=DefaultMethod
      drivenName=Pose_16
      functionType=DefaultFunction
      targetScale=0.95
      boneTransforms [
       {targetRotation=<rot -15.4723 -24.8915 12.5038>, targetTranslation=<-15.5111 0 0>}
       {targetRotation=<rot -64.545 87.1478 99.0891>, targetTranslation=<4.3575 -0.1466 2.3223>}
      ]
      customCurve {defaultValue=3.4028234663852886e+38, postInfinityExtrap=RCCE_Constant, preInfinityExtrap=RCCE_Constant, keys []}
      targetRotation=<rot 0 0 0>
     }
     {
      bApplyCustomCurve=false
      bIsHidden=false
      distanceMethod=DefaultMethod
      drivenName=Pose_17
      functionType=DefaultFunction
      targetScale=0.6058
      boneTransforms [
       {targetRotation=<rot -49.1322 -60.157 70.0006>, targetTranslation=<-15.5111 0 0>}
       {targetRotation=<rot -85.1916 41.4356 143.064>, targetTranslation=<4.3575 -0.1466 2.3223>}
      ]
      customCurve {defaultValue=3.4028234663852886e+38, postInfinityExtrap=RCCE_Constant, preInfinityExtrap=RCCE_Constant, keys []}
      targetRotation=<rot 0 0 0>
     }
    ]
    rBFParams {distanceMethod=Quaternion, function=Linear, radius=50}
    sourceBones [{boneName=upperarm_r}, {boneName=clavicle_r}]
   }
   n2 = Get arm_side_lift_r() -> (arm_side_lift_r:double)
   n3 = Get skydiving_animdyn_force() -> (skydiving_animdyn_force:Vector)
   n4 = Get tassle_damping() -> (tassle_damping:double)
   n5 = Use cached pose 'no_rigidbody_cache'()
   n6 = Use cached pose 'rigidbody_cache'()
   n7 = OR Boolean(A=NOT Boolean(A=Get bRigidBodyEnabled()), B="integer > integer"(A=Get CurrentLOD(), B=1) {is=title K2Node_CallFunction})
   n8 = "Constraint - dyn_tassel_01_r (P)"!(
    ComponentPose=Local To Component(LocalPose=Blend Poses by bool(BlendPose_0=@n5, BlendPose_1=@n6, bActiveValue=@n7) {blendType=Linear})
    Alpha=1
    ConstraintWeights_0=1
    ConstraintWeights_1:float=0
   ) -> (Pose) {
    alphaBoolBlend {blendOption=Linear}
    boneToModify {boneName=dyn_tassel_01_r}
    constraintSetup [{targetBone {boneName=pd_shoulderPad_r}}]
    constraintWeights [1,0]
   }
   n9 = "Constraint - dyn_tassel_02_r (PPP)"!(ComponentPose=@n8, Alpha=1, ConstraintWeights_0=0.0131, ConstraintWeights_1:float=0.0014, ConstraintWeights_2:float=0.9854) -> (Pose) {
    alphaBoolBlend {blendOption=Linear}
    boneToModify {boneName=dyn_tassel_02_r}
    constraintSetup [{targetBone {boneName=dyn_shoulderPad_fr_r}}, {targetBone {boneName=spine_05}}, {targetBone {boneName=pd_shoulderPad_r}}]
    constraintWeights [1,1,1]
   }
   n10 = "Constraint - dyn_tassel_03_r (PP)"!(ComponentPose=@n9, Alpha=1, ConstraintWeights_0=0.0877, ConstraintWeights_1:float=0.9123) -> (Pose) {
    alphaBoolBlend {blendOption=Linear}
    boneToModify {boneName=dyn_tassel_03_r}
    constraintSetup [{targetBone {boneName=dyn_shoulderPad_fr_r}}, {targetBone {boneName=pd_shoulderPad_r}}]
    constraintWeights [1,1]
   }
   n11 = "Constraint - dyn_tassel_04_r (PP)"!(ComponentPose=@n10, Alpha=1, ConstraintWeights_0=0.3939, ConstraintWeights_1:float=0.6061) -> (Pose) {
    alphaBoolBlend {blendOption=Linear}
    boneToModify {boneName=dyn_tassel_04_r}
    constraintSetup [{targetBone {boneName=dyn_shoulderPad_fr_r}}, {targetBone {boneName=pd_shoulderPad_r}}]
    constraintWeights [1,1]
   }
   n12 = "Constraint - dyn_tassel_05_r (PP)"!(ComponentPose=@n11, Alpha=1, ConstraintWeights_0=0.6056, ConstraintWeights_1:float=0.3944) -> (Pose) {
    alphaBoolBlend {blendOption=Linear}
    boneToModify {boneName=dyn_tassel_05_r}
    constraintSetup [{targetBone {boneName=dyn_shoulderPad_fr_r}}, {targetBone {boneName=pd_shoulderPad_r}}]
    constraintWeights [1,1]
   }
   n13 = "Constraint - dyn_tassel_06_r (PPPP)"!(
    ComponentPose=@n12
    Alpha=1
    ConstraintWeights_0=0.8725
    ConstraintWeights_1:float=0.0014
    ConstraintWeights_2:float=0.0013
    ConstraintWeights_3:float=0.1248
   ) -> (Pose) {
    alphaBoolBlend {blendOption=Linear}
    boneToModify {boneName=dyn_tassel_06_r}
    constraintSetup [
     {targetBone {boneName=dyn_shoulderPad_fr_r}}
     {targetBone {boneName=spine_03}}
     {targetBone {boneName=spine_04}}
     {targetBone {boneName=pd_shoulderPad_r}}
    ]
    constraintWeights [1,1,1,1]
   }
   n14 = "Constraint - dyn_tassel_07_r (PPPPPPP)"!(
    ComponentPose=@n13
    Alpha=1
    ConstraintWeights_0=0.0337
    ConstraintWeights_1:float=0.0046
    ConstraintWeights_2:float=0.017
    ConstraintWeights_3:float=0.029
    ConstraintWeights_4:float=0.0294
    ConstraintWeights_5:float=0.0046
    ConstraintWeights_6:float=0.8818
   ) -> (Pose) {
    alphaBoolBlend {blendOption=Linear}
    boneToModify {boneName=dyn_tassel_07_r}
    constraintSetup [
     {targetBone {boneName=pd_shoulderPad_r}}
     {targetBone {boneName=dyn_shoulderpad_r}}
     {targetBone {boneName=spine_02}}
     {targetBone {boneName=spine_03}}
     {targetBone {boneName=spine_04}}
     {targetBone {boneName=spine_05}}
     {targetBone {boneName=dyn_shoulderPad_fr_r}}
    ]
    constraintWeights [1,1,1,1,1,1,1]
   }
   n15 = "Constraint - dyn_tassel_01_l (P)"!(ComponentPose=@n14, Alpha=1, ConstraintWeights_0=1, ConstraintWeights_1:float=0) -> (Pose) {
    alphaBoolBlend {blendOption=Linear}
    boneToModify {boneName=dyn_tassel_01_l}
    constraintSetup [{targetBone {boneName=pd_shoulderPad_l}}]
    constraintWeights [1,0]
   }
   n16 = "Constraint - dyn_tassel_02_l (PPP)"!(ComponentPose=@n15, Alpha=1, ConstraintWeights_0=0.0131, ConstraintWeights_1:float=0.0014, ConstraintWeights_2:float=0.9854) -> (Pose) {
    alphaBoolBlend {blendOption=Linear}
    boneToModify {boneName=dyn_tassel_02_l}
    constraintSetup [{targetBone {boneName=dyn_shoulderPad_fr_l}}, {targetBone {boneName=spine_05}}, {targetBone {boneName=pd_shoulderPad_l}}]
    constraintWeights [1,1,1]
   }
   n17 = "Constraint - dyn_tassel_03_l (PP)"!(ComponentPose=@n16, Alpha=1, ConstraintWeights_0=0.0877, ConstraintWeights_1:float=0.9123) -> (Pose) {
    alphaBoolBlend {blendOption=Linear}
    boneToModify {boneName=dyn_tassel_03_l}
    constraintSetup [{targetBone {boneName=dyn_shoulderPad_fr_l}}, {targetBone {boneName=pd_shoulderPad_l}}]
    constraintWeights [1,1]
   }
   n18 = "Constraint - dyn_tassel_04_l (PP)"!(ComponentPose=@n17, Alpha=1, ConstraintWeights_0=0.3939, ConstraintWeights_1:float=0.6061) -> (Pose) {
    alphaBoolBlend {blendOption=Linear}
    boneToModify {boneName=dyn_tassel_04_l}
    constraintSetup [{targetBone {boneName=dyn_shoulderPad_fr_l}}, {targetBone {boneName=pd_shoulderPad_l}}]
    constraintWeights [1,1]
   }
   n19 = "Constraint - dyn_tassel_05_l (PP)"!(ComponentPose=@n18, Alpha=1, ConstraintWeights_0=0.6056, ConstraintWeights_1:float=0.3944) -> (Pose) {
    alphaBoolBlend {blendOption=Linear}
    boneToModify {boneName=dyn_tassel_05_l}
    constraintSetup [{targetBone {boneName=dyn_shoulderPad_fr_l}}, {targetBone {boneName=pd_shoulderPad_l}}]
    constraintWeights [1,1]
   }
   n20 = "Constraint - dyn_tassel_06_l (PPPP)"!(
    ComponentPose=@n19
    Alpha=1
    ConstraintWeights_0=0.8725
    ConstraintWeights_1:float=0.0014
    ConstraintWeights_2:float=0.0013
    ConstraintWeights_3:float=0.1248
   ) -> (Pose) {
    alphaBoolBlend {blendOption=Linear}
    boneToModify {boneName=dyn_tassel_06_l}
    constraintSetup [
     {targetBone {boneName=dyn_shoulderPad_fr_l}}
     {targetBone {boneName=spine_03}}
     {targetBone {boneName=spine_04}}
     {targetBone {boneName=pd_shoulderPad_l}}
    ]
    constraintWeights [1,1,1,1]
   }
   n21 = "Constraint - dyn_tassel_07_l (PPPPPPP)"!(
    ComponentPose=@n20
    Alpha=1
    ConstraintWeights_0=0.0337
    ConstraintWeights_1:float=0.0046
    ConstraintWeights_2:float=0.017
    ConstraintWeights_3:float=0.029
    ConstraintWeights_4:float=0.0294
    ConstraintWeights_5:float=0.0046
    ConstraintWeights_6:float=0.8818
   ) -> (Pose) {
    alphaBoolBlend {blendOption=Linear}
    boneToModify {boneName=dyn_tassel_07_l}
    constraintSetup [
     {targetBone {boneName=pd_shoulderPad_l}}
     {targetBone {boneName=dyn_shoulderpad_l}}
     {targetBone {boneName=spine_02}}
     {targetBone {boneName=spine_03}}
     {targetBone {boneName=spine_04}}
     {targetBone {boneName=spine_05}}
     {targetBone {boneName=dyn_shoulderPad_fr_l}}
    ]
    constraintWeights [1,1,1,1,1,1,1]
   }
   n22 = Anim Dynamics - Bone: dyn_tassel_01_r!(ComponentPose=@n21, Alpha=1, AngularDampingOverride:float=@n4, ExternalForce=@n3) -> (Pose) {
    angularDampingOverride=1
    angularSpringConstant=100
    bGravityOverrideInSimSpace=true
    bOverrideAngularDamping=true
    gravityScale=0.5
    simulationSpace=RootRelative
    alphaBoolBlend {blendOption=Linear}
    boundBone {boneName=dyn_tassel_01_r}
    componentAppliedLinearAccClamp=<3000 3000 3000>
    physicsBodyDefinitions [{
      sphereCollisionRadius=0
      boundBone {boneName=dyn_tassel_01_r}
      boxExtents=<5 10 10>
      constraintSetup {angularLimitsMax=<0 10 0>, angularLimitsMin=<0 -10 -30>, angularTarget=<0 0 0>}
      localJointOffset=<5 0 0>
     }]
   }
   n23 = Anim Dynamics - Bone: dyn_tassel_02_r!(ComponentPose=@n22, Alpha=1, AngularDampingOverride:float=@n4, ExternalForce=@n3) -> (Pose) {
    angularDampingOverride=1
    angularSpringConstant=100
    bGravityOverrideInSimSpace=true
    bOverrideAngularDamping=true
    gravityScale=0.5
    simulationSpace=RootRelative
    alphaBoolBlend {blendOption=Linear}
    boundBone {boneName=dyn_tassel_02_r}
    componentAppliedLinearAccClamp=<3000 3000 3000>
    physicsBodyDefinitions [{
      sphereCollisionRadius=0
      boundBone {boneName=dyn_tassel_02_r}
      boxExtents=<5 10 10>
      constraintSetup {angularLimitsMax=<0 10 0>, angularLimitsMin=<0 -10 -30>, angularTarget=<0 0 0>}
      localJointOffset=<5 0 0>
     }]
   }
   n24 = Anim Dynamics - Bone: dyn_tassel_03_r!(ComponentPose=@n23, Alpha=1, AngularDampingOverride:float=@n4, ExternalForce=@n3) -> (Pose) {
    angularDampingOverride=1
    angularSpringConstant=100
    bGravityOverrideInSimSpace=true
    bOverrideAngularDamping=true
    gravityScale=0.5
    simulationSpace=RootRelative
    alphaBoolBlend {blendOption=Linear}
    boundBone {boneName=dyn_tassel_03_r}
    componentAppliedLinearAccClamp=<3000 3000 3000>
    physicsBodyDefinitions [{
      sphereCollisionRadius=0
      boundBone {boneName=dyn_tassel_03_r}
      boxExtents=<5 10 10>
      constraintSetup {angularLimitsMax=<0 10 0>, angularLimitsMin=<0 -10 -30>, angularTarget=<0 0 0>}
      localJointOffset=<5 0 0>
     }]
   }
   n25 = Anim Dynamics - Bone: dyn_tassel_04_r!(ComponentPose=@n24, Alpha=1, AngularDampingOverride:float=@n4, ExternalForce=@n3) -> (Pose) {
    angularDampingOverride=1
    angularSpringConstant=100
    bGravityOverrideInSimSpace=true
    bOverrideAngularDamping=true
    gravityScale=0.5
    simulationSpace=RootRelative
    alphaBoolBlend {blendOption=Linear}
    boundBone {boneName=dyn_tassel_04_r}
    componentAppliedLinearAccClamp=<3000 3000 3000>
    physicsBodyDefinitions [{
      sphereCollisionRadius=0
      boundBone {boneName=dyn_tassel_04_r}
      boxExtents=<5 10 10>
      constraintSetup {angularLimitsMax=<0 10 0>, angularLimitsMin=<0 -10 -30>, angularTarget=<0 0 0>}
      localJointOffset=<5 0 0>
     }]
   }
   n26 = Anim Dynamics - Bone: dyn_tassel_05_r!(ComponentPose=@n25, Alpha=1, AngularDampingOverride:float=@n4, ExternalForce=@n3) -> (Pose) {
    angularDampingOverride=1
    angularSpringConstant=100
    bGravityOverrideInSimSpace=true
    bOverrideAngularDamping=true
    gravityScale=0.5
    simulationSpace=RootRelative
    alphaBoolBlend {blendOption=Linear}
    boundBone {boneName=dyn_tassel_05_r}
    componentAppliedLinearAccClamp=<3000 3000 3000>
    physicsBodyDefinitions [{
      sphereCollisionRadius=0
      boundBone {boneName=dyn_tassel_05_r}
      boxExtents=<5 10 10>
      constraintSetup {angularLimitsMax=<0 10 0>, angularLimitsMin=<0 -10 -30>, angularTarget=<0 0 0>}
      localJointOffset=<5 0 0>
     }]
   }
   n27 = Anim Dynamics - Bone: dyn_tassel_06_r!(ComponentPose=@n26, Alpha=1, AngularDampingOverride:float=@n4, ExternalForce=@n3) -> (Pose) {
    angularDampingOverride=1
    angularSpringConstant=100
    bGravityOverrideInSimSpace=true
    bOverrideAngularDamping=true
    gravityScale=0.5
    simulationSpace=RootRelative
    alphaBoolBlend {blendOption=Linear}
    boundBone {boneName=dyn_tassel_06_r}
    componentAppliedLinearAccClamp=<3000 3000 3000>
    physicsBodyDefinitions [{
      sphereCollisionRadius=0
      boundBone {boneName=dyn_tassel_06_r}
      boxExtents=<5 10 10>
      constraintSetup {angularLimitsMax=<0 10 0>, angularLimitsMin=<0 -10 -30>, angularTarget=<0 0 0>}
      localJointOffset=<5 0 0>
     }]
   }
   n28 = Anim Dynamics - Bone: dyn_tassel_sub_01_r!(ComponentPose=@n27, Alpha=1, AngularDampingOverride:float=@n4, ExternalForce="0,0,0") -> (Pose) {
    angularDampingOverride=1
    angularSpringConstant=100
    bGravityOverrideInSimSpace=true
    bOverrideAngularDamping=true
    gravityScale=0.5
    simulationSpace=RootRelative
    alphaBoolBlend {blendOption=Linear}
    boundBone {boneName=dyn_tassel_sub_01_r}
    componentAppliedLinearAccClamp=<3000 3000 3000>
    physicsBodyDefinitions [{
      sphereCollisionRadius=0
      boundBone {boneName=dyn_tassel_sub_01_r}
      boxExtents=<2 20 20>
      constraintSetup {angularLimitsMax=<0 10 0>, angularLimitsMin=<0 -10 -30>, angularTarget=<0 0 0>}
      localJointOffset=<5 0 0>
     }]
   }
   n29 = Anim Dynamics - Bone: dyn_tassel_sub_02_r!(ComponentPose=@n28, Alpha=1, AngularDampingOverride:float=@n4, ExternalForce="0,0,0") -> (Pose) {
    angularDampingOverride=1
    angularSpringConstant=100
    bGravityOverrideInSimSpace=true
    bOverrideAngularDamping=true
    gravityScale=0.5
    simulationSpace=RootRelative
    alphaBoolBlend {blendOption=Linear}
    boundBone {boneName=dyn_tassel_sub_02_r}
    componentAppliedLinearAccClamp=<3000 3000 3000>
    physicsBodyDefinitions [{
      sphereCollisionRadius=0
      boundBone {boneName=dyn_tassel_sub_02_r}
      boxExtents=<2 20 20>
      constraintSetup {angularLimitsMax=<0 10 0>, angularLimitsMin=<0 -10 -30>, angularTarget=<0 0 0>}
      localJointOffset=<5 0 0>
     }]
   }
   n30 = Anim Dynamics - Bone: dyn_tassel_sub_03_r!(ComponentPose=@n29, Alpha=1, AngularDampingOverride:float=@n4, ExternalForce="0,0,0") -> (Pose) {
    angularDampingOverride=1
    angularSpringConstant=100
    bGravityOverrideInSimSpace=true
    bOverrideAngularDamping=true
    gravityScale=0.5
    simulationSpace=RootRelative
    alphaBoolBlend {blendOption=Linear}
    boundBone {boneName=dyn_tassel_sub_03_r}
    componentAppliedLinearAccClamp=<3000 3000 3000>
    physicsBodyDefinitions [{
      sphereCollisionRadius=0
      boundBone {boneName=dyn_tassel_sub_03_r}
      boxExtents=<2 20 20>
      constraintSetup {angularLimitsMax=<0 10 0>, angularLimitsMin=<0 -10 -30>, angularTarget=<0 0 0>}
      localJointOffset=<5 0 0>
     }]
   }
   n31 = Anim Dynamics - Bone: dyn_tassel_sub_04_r!(ComponentPose=@n30, Alpha=1, AngularDampingOverride:float=@n4, ExternalForce="0,0,0") -> (Pose) {
    angularDampingOverride=1
    angularSpringConstant=100
    bGravityOverrideInSimSpace=true
    bOverrideAngularDamping=true
    gravityScale=0.5
    simulationSpace=RootRelative
    alphaBoolBlend {blendOption=Linear}
    boundBone {boneName=dyn_tassel_sub_04_r}
    componentAppliedLinearAccClamp=<3000 3000 3000>
    physicsBodyDefinitions [{
      sphereCollisionRadius=0
      boundBone {boneName=dyn_tassel_sub_04_r}
      boxExtents=<2 20 20>
      constraintSetup {angularLimitsMax=<0 10 0>, angularLimitsMin=<0 -10 -30>, angularTarget=<0 0 0>}
      localJointOffset=<5 0 0>
     }]
   }
   n32 = Anim Dynamics - Bone: dyn_tassel_sub_05_r!(ComponentPose=@n31, Alpha=1, AngularDampingOverride:float=@n4, ExternalForce="0,0,0") -> (Pose) {
    angularDampingOverride=1
    angularSpringConstant=100
    bGravityOverrideInSimSpace=true
    bOverrideAngularDamping=true
    gravityScale=0.5
    simulationSpace=RootRelative
    alphaBoolBlend {blendOption=Linear}
    boundBone {boneName=dyn_tassel_sub_05_r}
    componentAppliedLinearAccClamp=<3000 3000 3000>
    physicsBodyDefinitions [{
      sphereCollisionRadius=0
      boundBone {boneName=dyn_tassel_sub_05_r}
      boxExtents=<2 20 20>
      constraintSetup {angularLimitsMax=<0 10 0>, angularLimitsMin=<0 -10 -30>, angularTarget=<0 0 0>}
      localJointOffset=<5 0 0>
     }]
   }
   n33 = Anim Dynamics - Bone: dyn_tassel_sub_06_r!(ComponentPose=@n32, Alpha=1, AngularDampingOverride:float=@n4, ExternalForce="0,0,0") -> (Pose) {
    angularDampingOverride=1
    angularSpringConstant=100
    bGravityOverrideInSimSpace=true
    bOverrideAngularDamping=true
    gravityScale=0.5
    simulationSpace=RootRelative
    alphaBoolBlend {blendOption=Linear}
    boundBone {boneName=dyn_tassel_sub_06_r}
    componentAppliedLinearAccClamp=<3000 3000 3000>
    physicsBodyDefinitions [{
      sphereCollisionRadius=0
      boundBone {boneName=dyn_tassel_sub_06_r}
      boxExtents=<2 20 20>
      constraintSetup {angularLimitsMax=<0 10 0>, angularLimitsMin=<0 -10 -30>, angularTarget=<0 0 0>}
      localJointOffset=<5 0 0>
     }]
   }
   n34 = Get skydiving_animdyn_force()
   n35 = Get tassle_damping()
   n36 = Anim Dynamics - Bone: dyn_tassel_01_l!(ComponentPose=@n33, Alpha=1, AngularDampingOverride:float=@n35, ExternalForce=@n34) -> (Pose) {
    angularDampingOverride=1
    angularSpringConstant=100
    bGravityOverrideInSimSpace=true
    bOverrideAngularDamping=true
    gravityScale=0.5
    simulationSpace=RootRelative
    alphaBoolBlend {blendOption=Linear}
    boundBone {boneName=dyn_tassel_01_l}
    componentAppliedLinearAccClamp=<3000 3000 3000>
    physicsBodyDefinitions [{
      sphereCollisionRadius=0
      boundBone {boneName=dyn_tassel_01_l}
      boxExtents=<5 10 10>
      constraintSetup {angularLimitsMax=<0 10 0>, angularLimitsMin=<0 -10 -30>, angularTarget=<0 0 0>}
      localJointOffset=<5 0 0>
     }]
   }
   n37 = Anim Dynamics - Bone: dyn_tassel_02_l!(ComponentPose=@n36, Alpha=1, AngularDampingOverride:float=@n35, ExternalForce=@n34) -> (Pose) {
    angularDampingOverride=1
    angularSpringConstant=100
    bGravityOverrideInSimSpace=true
    bOverrideAngularDamping=true
    gravityScale=0.5
    simulationSpace=RootRelative
    alphaBoolBlend {blendOption=Linear}
    boundBone {boneName=dyn_tassel_02_l}
    componentAppliedLinearAccClamp=<3000 3000 3000>
    physicsBodyDefinitions [{
      sphereCollisionRadius=0
      boundBone {boneName=dyn_tassel_02_l}
      boxExtents=<5 10 10>
      constraintSetup {angularLimitsMax=<0 10 0>, angularLimitsMin=<0 -10 -30>, angularTarget=<0 0 0>}
      localJointOffset=<5 0 0>
     }]
   }
   n38 = Anim Dynamics - Bone: dyn_tassel_03_l!(ComponentPose=@n37, Alpha=1, AngularDampingOverride:float=@n35, ExternalForce=@n34) -> (Pose) {
    angularDampingOverride=1
    angularSpringConstant=100
    bGravityOverrideInSimSpace=true
    bOverrideAngularDamping=true
    gravityScale=0.5
    simulationSpace=RootRelative
    alphaBoolBlend {blendOption=Linear}
    boundBone {boneName=dyn_tassel_03_l}
    componentAppliedLinearAccClamp=<3000 3000 3000>
    physicsBodyDefinitions [{
      sphereCollisionRadius=0
      boundBone {boneName=dyn_tassel_03_l}
      boxExtents=<5 10 10>
      constraintSetup {angularLimitsMax=<0 10 0>, angularLimitsMin=<0 -10 -30>, angularTarget=<0 0 0>}
      localJointOffset=<5 0 0>
     }]
   }
   n39 = Anim Dynamics - Bone: dyn_tassel_04_l!(ComponentPose=@n38, Alpha=1, AngularDampingOverride:float=@n35, ExternalForce=@n34) -> (Pose) {
    angularDampingOverride=1
    angularSpringConstant=100
    bGravityOverrideInSimSpace=true
    bOverrideAngularDamping=true
    gravityScale=0.5
    simulationSpace=RootRelative
    alphaBoolBlend {blendOption=Linear}
    boundBone {boneName=dyn_tassel_04_l}
    componentAppliedLinearAccClamp=<3000 3000 3000>
    physicsBodyDefinitions [{
      sphereCollisionRadius=0
      boundBone {boneName=dyn_tassel_04_l}
      boxExtents=<5 10 10>
      constraintSetup {angularLimitsMax=<0 10 0>, angularLimitsMin=<0 -10 -30>, angularTarget=<0 0 0>}
      localJointOffset=<5 0 0>
     }]
   }
   n40 = Anim Dynamics - Bone: dyn_tassel_05_l!(ComponentPose=@n39, Alpha=1, AngularDampingOverride:float=@n35, ExternalForce=@n34) -> (Pose) {
    angularDampingOverride=1
    angularSpringConstant=100
    bGravityOverrideInSimSpace=true
    bOverrideAngularDamping=true
    gravityScale=0.5
    simulationSpace=RootRelative
    alphaBoolBlend {blendOption=Linear}
    boundBone {boneName=dyn_tassel_05_l}
    componentAppliedLinearAccClamp=<3000 3000 3000>
    physicsBodyDefinitions [{
      sphereCollisionRadius=0
      boundBone {boneName=dyn_tassel_05_l}
      boxExtents=<5 10 10>
      constraintSetup {angularLimitsMax=<0 10 0>, angularLimitsMin=<0 -10 -30>, angularTarget=<0 0 0>}
      localJointOffset=<5 0 0>
     }]
   }
   n41 = Anim Dynamics - Bone: dyn_tassel_06_l!(ComponentPose=@n40, Alpha=1, AngularDampingOverride:float=@n35, ExternalForce=@n34) -> (Pose) {
    angularDampingOverride=1
    angularSpringConstant=100
    bGravityOverrideInSimSpace=true
    bOverrideAngularDamping=true
    gravityScale=0.5
    simulationSpace=RootRelative
    alphaBoolBlend {blendOption=Linear}
    boundBone {boneName=dyn_tassel_06_l}
    componentAppliedLinearAccClamp=<3000 3000 3000>
    physicsBodyDefinitions [{
      sphereCollisionRadius=0
      boundBone {boneName=dyn_tassel_06_l}
      boxExtents=<5 10 10>
      constraintSetup {angularLimitsMax=<0 10 0>, angularLimitsMin=<0 -10 -30>, angularTarget=<0 0 0>}
      localJointOffset=<5 0 0>
     }]
   }
   n42 = Anim Dynamics - Bone: dyn_tassel_sub_01_l!(ComponentPose=@n41, Alpha=1, AngularDampingOverride:float=@n35, ExternalForce="0,0,0") -> (Pose) {
    angularDampingOverride=1
    angularSpringConstant=100
    bGravityOverrideInSimSpace=true
    bOverrideAngularDamping=true
    gravityScale=0.5
    simulationSpace=RootRelative
    alphaBoolBlend {blendOption=Linear}
    boundBone {boneName=dyn_tassel_sub_01_l}
    componentAppliedLinearAccClamp=<3000 3000 3000>
    physicsBodyDefinitions [{
      sphereCollisionRadius=0
      boundBone {boneName=dyn_tassel_sub_01_l}
      boxExtents=<5 10 10>
      constraintSetup {angularLimitsMax=<0 10 0>, angularLimitsMin=<0 -10 -30>, angularTarget=<0 0 0>}
      localJointOffset=<5 0 0>
     }]
   }
   n43 = Anim Dynamics - Bone: dyn_tassel_sub_02_l!(ComponentPose=@n42, Alpha=1, AngularDampingOverride:float=@n35, ExternalForce="0,0,0") -> (Pose) {
    angularDampingOverride=1
    angularSpringConstant=100
    bGravityOverrideInSimSpace=true
    bOverrideAngularDamping=true
    gravityScale=0.5
    simulationSpace=RootRelative
    alphaBoolBlend {blendOption=Linear}
    boundBone {boneName=dyn_tassel_sub_02_l}
    componentAppliedLinearAccClamp=<3000 3000 3000>
    physicsBodyDefinitions [{
      sphereCollisionRadius=0
      boundBone {boneName=dyn_tassel_sub_02_l}
      boxExtents=<5 10 10>
      constraintSetup {angularLimitsMax=<0 10 0>, angularLimitsMin=<0 -10 -30>, angularTarget=<0 0 0>}
      localJointOffset=<5 0 0>
     }]
   }
   n44 = Anim Dynamics - Bone: dyn_tassel_sub_03_l!(ComponentPose=@n43, Alpha=1, AngularDampingOverride:float=@n35, ExternalForce="0,0,0") -> (Pose) {
    angularDampingOverride=1
    angularSpringConstant=100
    bGravityOverrideInSimSpace=true
    bOverrideAngularDamping=true
    gravityScale=0.5
    simulationSpace=RootRelative
    alphaBoolBlend {blendOption=Linear}
    boundBone {boneName=dyn_tassel_sub_03_l}
    componentAppliedLinearAccClamp=<3000 3000 3000>
    physicsBodyDefinitions [{
      sphereCollisionRadius=0
      boundBone {boneName=dyn_tassel_sub_03_l}
      boxExtents=<5 10 10>
      constraintSetup {angularLimitsMax=<0 10 0>, angularLimitsMin=<0 -10 -30>, angularTarget=<0 0 0>}
      localJointOffset=<5 0 0>
     }]
   }
   n45 = Anim Dynamics - Bone: dyn_tassel_sub_04_l!(ComponentPose=@n44, Alpha=1, AngularDampingOverride:float=@n35, ExternalForce="0,0,0") -> (Pose) {
    angularDampingOverride=1
    angularSpringConstant=100
    bGravityOverrideInSimSpace=true
    bOverrideAngularDamping=true
    gravityScale=0.5
    simulationSpace=RootRelative
    alphaBoolBlend {blendOption=Linear}
    boundBone {boneName=dyn_tassel_sub_04_l}
    componentAppliedLinearAccClamp=<3000 3000 3000>
    physicsBodyDefinitions [{
      sphereCollisionRadius=0
      boundBone {boneName=dyn_tassel_sub_04_l}
      boxExtents=<5 10 10>
      constraintSetup {angularLimitsMax=<0 10 0>, angularLimitsMin=<0 -10 -30>, angularTarget=<0 0 0>}
      localJointOffset=<5 0 0>
     }]
   }
   n46 = Anim Dynamics - Bone: dyn_tassel_sub_05_l!(ComponentPose=@n45, Alpha=1, AngularDampingOverride:float=@n35, ExternalForce="0,0,0") -> (Pose) {
    angularDampingOverride=1
    angularSpringConstant=100
    bGravityOverrideInSimSpace=true
    bOverrideAngularDamping=true
    gravityScale=0.5
    simulationSpace=RootRelative
    alphaBoolBlend {blendOption=Linear}
    boundBone {boneName=dyn_tassel_sub_05_l}
    componentAppliedLinearAccClamp=<3000 3000 3000>
    physicsBodyDefinitions [{
      sphereCollisionRadius=0
      boundBone {boneName=dyn_tassel_sub_05_l}
      boxExtents=<5 10 10>
      constraintSetup {angularLimitsMax=<0 10 0>, angularLimitsMin=<0 -10 -30>, angularTarget=<0 0 0>}
      localJointOffset=<5 0 0>
     }]
   }
   n47 = Anim Dynamics - Bone: dyn_tassel_sub_06_l!(ComponentPose=@n46, Alpha=1, AngularDampingOverride:float=@n35, ExternalForce="0,0,0") -> (Pose) {
    angularDampingOverride=1
    angularSpringConstant=100
    bGravityOverrideInSimSpace=true
    bOverrideAngularDamping=true
    gravityScale=0.5
    simulationSpace=RootRelative
    alphaBoolBlend {blendOption=Linear}
    boundBone {boneName=dyn_tassel_sub_06_l}
    componentAppliedLinearAccClamp=<3000 3000 3000>
    physicsBodyDefinitions [{
      sphereCollisionRadius=0
      boundBone {boneName=dyn_tassel_sub_06_l}
      boxExtents=<5 10 10>
      constraintSetup {angularLimitsMax=<0 10 0>, angularLimitsMin=<0 -10 -30>, angularTarget=<0 0 0>}
      localJointOffset=<5 0 0>
     }]
   }
   n48 = Anim Dynamics - Bone: dyn_pegleg_cloth_T3(ComponentPose=@n47, Alpha=1, ExternalForce="0,0,0") {
    angularDampingOverride=1
    bGravityOverrideInSimSpace=true
    bOverrideAngularDamping=true
    gravityScale=0.35
    alphaBoolBlend {blendOption=Linear}
    boundBone {boneName=dyn_pegleg_cloth_T3}
    componentLinearAccScale=<1 1 1>
    componentLinearVelScale=<1 1 1>
    physicsBodyDefinitions [{
      sphereCollisionRadius=0
      boundBone {boneName=dyn_pegleg_cloth_T3}
      boxExtents=<2 10 10>
      constraintSetup {angularLimitsMax=<0 20 10>, angularLimitsMin=<0 -20 -10>, angularTarget=<0 0 0>}
      localJointOffset=<2 0 -0.1>
     }]
   }
   n49 = Anim Dynamics - Bone: dyn_pegleg_cloth_T4_01(ComponentPose=@n48, Alpha=1, ExternalForce="0,0,0") {
    angularDampingOverride=1
    bGravityOverrideInSimSpace=true
    bOverrideAngularDamping=true
    gravityScale=0.35
    simulationSpace=RootRelative
    alphaBoolBlend {blendOption=Linear}
    boundBone {boneName=dyn_pegleg_cloth_T4_01}
    physicsBodyDefinitions [{
      sphereCollisionRadius=0
      boundBone {boneName=dyn_pegleg_cloth_T4_01}
      boxExtents=<2 10 10>
      constraintSetup {angularLimitsMax=<0 0 20>, angularLimitsMin=<0 0 -20>, angularTarget=<0 0 0>}
      localJointOffset=<2 0 0>
     }]
   }
   n50 = Anim Dynamics - Bone: dyn_pegleg_cloth_T4_02(ComponentPose=@n49, Alpha=1, ExternalForce="0,0,0") {
    angularDampingOverride=1
    bGravityOverrideInSimSpace=true
    bOverrideAngularDamping=true
    gravityScale=0.35
    simulationSpace=RootRelative
    alphaBoolBlend {blendOption=Linear}
    boundBone {boneName=dyn_pegleg_cloth_T4_02}
    physicsBodyDefinitions [{
      sphereCollisionRadius=0
      boundBone {boneName=dyn_pegleg_cloth_T4_02}
      boxExtents=<2 10 10>
      constraintSetup {angularLimitsMax=<0 0 20>, angularLimitsMin=<0 0 -20>, angularTarget=<0 0 0>}
      localJointOffset=<2 0 0>
     }]
   }
   n51 = Anim Dynamics - Bone: dyn_pegleg_cloth_T4_03(ComponentPose=@n50, Alpha=1, ExternalForce="0,0,0") {
    angularDampingOverride=1
    bGravityOverrideInSimSpace=true
    bOverrideAngularDamping=true
    gravityScale=0.35
    simulationSpace=RootRelative
    alphaBoolBlend {blendOption=Linear}
    boundBone {boneName=dyn_pegleg_cloth_T4_03}
    physicsBodyDefinitions [{
      sphereCollisionRadius=0
      boundBone {boneName=dyn_pegleg_cloth_T4_03}
      boxExtents=<2 10 10>
      constraintSetup {angularLimitsMax=<0 0 20>, angularLimitsMin=<0 0 -20>, angularTarget=<0 0 0>}
      localJointOffset=<2 0 0>
     }]
   }
   n52 = Anim Dynamics - Bone: dyn_pegleg_cloth_T4_04(ComponentPose=@n51, Alpha=1, ExternalForce="0,0,0") {
    angularDampingOverride=1
    bGravityOverrideInSimSpace=true
    bOverrideAngularDamping=true
    gravityScale=0.35
    simulationSpace=RootRelative
    alphaBoolBlend {blendOption=Linear}
    boundBone {boneName=dyn_pegleg_cloth_T4_04}
    physicsBodyDefinitions [{
      sphereCollisionRadius=0
      boundBone {boneName=dyn_pegleg_cloth_T4_04}
      boxExtents=<2 10 10>
      constraintSetup {angularLimitsMax=<0 0 20>, angularLimitsMin=<0 0 -20>, angularTarget=<0 0 0>}
      localJointOffset=<2 0 0>
     }]
   }
   n53 = Look At - Bone: dyn_hook(ComponentPose=@n52, Alpha=1) {
    alphaBoolBlend {blendOption=Linear}
    boneToModify {boneName=dyn_hook}
    lookAtLocation {x=0}
    lookAtTarget {boneReference {boneName=middle_01_l}}
    lookAt_Axis {axis=<1 0>}
   }
   n54 = Anim Dynamics - Bone: dyn_necklace_T4_01(ComponentPose=@n53, Alpha=1, ExternalForce="0,0,0") {
    angularDampingOverride=1
    angularSpringConstant=100
    bGravityOverrideInSimSpace=true
    bOverrideAngularDamping=true
    gravityScale=0.5
    simulationSpace=RootRelative
    alphaBoolBlend {blendOption=Linear}
    boundBone {boneName=dyn_necklace_T4_01}
    componentAppliedLinearAccClamp=<3000 3000 3000>
    physicsBodyDefinitions [{
      sphereCollisionRadius=0
      boundBone {boneName=dyn_necklace_T4_01}
      boxExtents=<5 10 10>
      constraintSetup {angularLimitsMax=<0 10 30>, angularLimitsMin=<0 -10 5>, angularTarget=<0 0 0>}
      localJointOffset=<5 0 0>
     }]
   }
   n55 = Anim Dynamics - Bone: dyn_necklace_T4_02(ComponentPose=@n54, Alpha=1, ExternalForce="0,0,0") {
    angularDampingOverride=1
    angularSpringConstant=100
    bGravityOverrideInSimSpace=true
    bOverrideAngularDamping=true
    gravityScale=0.5
    simulationSpace=RootRelative
    alphaBoolBlend {blendOption=Linear}
    boundBone {boneName=dyn_necklace_T4_02}
    componentAppliedLinearAccClamp=<3000 3000 3000>
    physicsBodyDefinitions [{
      sphereCollisionRadius=0
      boundBone {boneName=dyn_necklace_T4_02}
      boxExtents=<5 10 10>
      constraintSetup {angularLimitsMax=<0 10 30>, angularLimitsMin=<0 -10 5>, angularTarget=<0 0 0>}
      localJointOffset=<5 0 0>
     }]
   }
   n56 = Anim Dynamics - Bone: dyn_necklace_T4_03(ComponentPose=@n55, Alpha=1, ExternalForce="0,0,0") {
    angularDampingOverride=1
    angularSpringConstant=100
    bGravityOverrideInSimSpace=true
    bOverrideAngularDamping=true
    gravityScale=0.5
    simulationSpace=RootRelative
    alphaBoolBlend {blendOption=Linear}
    boundBone {boneName=dyn_necklace_T4_03}
    componentAppliedLinearAccClamp=<3000 3000 3000>
    gravityOverride {z=-980}
    physicsBodyDefinitions [{
      sphereCollisionRadius=0
      boundBone {boneName=dyn_necklace_T4_03}
      boxExtents=<5 10 10>
      constraintSetup {angularLimitsMax=<0 10 30>, angularLimitsMin=<0 -10 5>, angularTarget=<0 0 0>}
      localJointOffset=<5 0 0>
     }]
   }
   n57 = "Transform (Modify) Bone - Bone: dyn_shoulderpad_r"(ComponentPose=@n56, Alpha=@n2, Translation="0,0,0", Rotation="10,0,0", Scale="1,1,1") {
    rotationMode=BMM_Additive
    rotationSpace=BCS_ParentBoneSpace
    alphaBoolBlend {blendOption=Linear}
    boneToModify {boneName=dyn_shoulderpad_r}
    rotation {pitch=10}
   }
   n58 = Get tassle_skydiving_force() -> (tassle_skydiving_force:Rotator)
   n59 = "Transform (Modify) Bone - Bone: dyn_tassel_01_r"(ComponentPose=@n57, Alpha=@n2, Translation="0,0,0", Rotation=@n58, Scale="1,1,1") {rotationMode=BMM_Additive, rotationSpace=BCS_ParentBoneSpace, alphaBoolBlend {blendOption=Linear}, boneToModify {boneName=dyn_tassel_01_r}}
   n60 = "Transform (Modify) Bone - Bone: dyn_tassel_02_r"(ComponentPose=@n59, Alpha=@n2, Translation="0,0,0", Rotation=@n58, Scale="1,1,1") {rotationMode=BMM_Additive, rotationSpace=BCS_ParentBoneSpace, alphaBoolBlend {blendOption=Linear}, boneToModify {boneName=dyn_tassel_02_r}}
   n61 = "Transform (Modify) Bone - Bone: dyn_tassel_03_r"(ComponentPose=@n60, Alpha=@n2, Translation="0,0,0", Rotation=@n58, Scale="1,1,1") {rotationMode=BMM_Additive, rotationSpace=BCS_ParentBoneSpace, alphaBoolBlend {blendOption=Linear}, boneToModify {boneName=dyn_tassel_03_r}}
   n62 = "Transform (Modify) Bone - Bone: dyn_tassel_04_r"(ComponentPose=@n61, Alpha=@n2, Translation="0,0,0", Rotation=@n58, Scale="1,1,1") {rotationMode=BMM_Additive, rotationSpace=BCS_ParentBoneSpace, alphaBoolBlend {blendOption=Linear}, boneToModify {boneName=dyn_tassel_04_r}}
   n63 = "Transform (Modify) Bone - Bone: dyn_tassel_05_r"(ComponentPose=@n62, Alpha=@n2, Translation="0,0,0", Rotation=@n58, Scale="1,1,1") {rotationMode=BMM_Additive, rotationSpace=BCS_ParentBoneSpace, alphaBoolBlend {blendOption=Linear}, boneToModify {boneName=dyn_tassel_05_r}}
   n64 = "Transform (Modify) Bone - Bone: dyn_tassel_06_r"(ComponentPose=@n63, Alpha=@n2, Translation="0,0,0", Rotation=@n58, Scale="1,1,1") {rotationMode=BMM_Additive, rotationSpace=BCS_ParentBoneSpace, alphaBoolBlend {blendOption=Linear}, boneToModify {boneName=dyn_tassel_06_r}}
   n65 = Get arm_side_lift_l() -> (arm_side_lift_l:double)
   n66 = "Transform (Modify) Bone - Bone: dyn_shoulderpad_l"(ComponentPose=@n64, Alpha=@n65, Translation="0,0,0", Rotation="10,0,0", Scale="1,1,1") {
    rotationMode=BMM_Additive
    rotationSpace=BCS_ParentBoneSpace
    alphaBoolBlend {blendOption=Linear}
    boneToModify {boneName=dyn_shoulderpad_l}
    rotation {pitch=10}
   }
   n67 = Get tassle_skydiving_force()
   n68 = "Transform (Modify) Bone - Bone: dyn_tassel_01_l"(ComponentPose=@n66, Alpha=@n65, Translation="0,0,0", Rotation=@n67, Scale="1,1,1") {rotationMode=BMM_Additive, rotationSpace=BCS_ParentBoneSpace, alphaBoolBlend {blendOption=Linear}, boneToModify {boneName=dyn_tassel_01_l}}
   n69 = "Transform (Modify) Bone - Bone: dyn_tassel_02_l"(ComponentPose=@n68, Alpha=@n65, Translation="0,0,0", Rotation=@n67, Scale="1,1,1") {rotationMode=BMM_Additive, rotationSpace=BCS_ParentBoneSpace, alphaBoolBlend {blendOption=Linear}, boneToModify {boneName=dyn_tassel_02_l}}
   n70 = "Transform (Modify) Bone - Bone: dyn_tassel_03_l"(ComponentPose=@n69, Alpha=@n65, Translation="0,0,0", Rotation=@n67, Scale="1,1,1") {rotationMode=BMM_Additive, rotationSpace=BCS_ParentBoneSpace, alphaBoolBlend {blendOption=Linear}, boneToModify {boneName=dyn_tassel_03_l}}
   n71 = "Transform (Modify) Bone - Bone: dyn_tassel_04_l"(ComponentPose=@n70, Alpha=@n65, Translation="0,0,0", Rotation=@n67, Scale="1,1,1") {rotationMode=BMM_Additive, rotationSpace=BCS_ParentBoneSpace, alphaBoolBlend {blendOption=Linear}, boneToModify {boneName=dyn_tassel_04_l}}
   n72 = "Transform (Modify) Bone - Bone: dyn_tassel_05_l"(ComponentPose=@n71, Alpha=@n65, Translation="0,0,0", Rotation=@n67, Scale="1,1,1") {rotationMode=BMM_Additive, rotationSpace=BCS_ParentBoneSpace, alphaBoolBlend {blendOption=Linear}, boneToModify {boneName=dyn_tassel_05_l}}
   n73 = "Transform (Modify) Bone - Bone: dyn_tassel_06_l"(ComponentPose=@n72, Alpha=@n65, Translation="0,0,0", Rotation=@n67, Scale="1,1,1") {rotationMode=BMM_Additive, rotationSpace=BCS_ParentBoneSpace, alphaBoolBlend {blendOption=Linear}, boneToModify {boneName=dyn_tassel_06_l}}
   Output Pose(Result=Component To Local(ComponentPose=@n73)) {layerGroup=None}
   Save cached pose 'copypose_cache'(Pose=Copy Pose From Mesh() {bUseAttachedParent=true}) {cacheName=copypose_cache}
   n74 = Flap Node - Bone: dyn_shoulderpad_l!(ComponentPose:Component Space Pose Link Structure=Local To Component(LocalPose=Use cached pose 'copypose_cache'()), Alpha:float=0.4) -> (Pose:Component Space Pose Link Structure) {
    is=title FortAnimGraphNode_Flap
    bUseClamp=true
    targetClampMax=1
    targetClampMin=-1
    targetComponent=RotationZ
    alphaBoolBlend {blendOption=Linear}
    sourceDrives [{
      bUseRange=false
      drivingCurve=null
      multiplier=-0.8
      rangeMax=1
      rangeMin=-1
      remappedMax=1
      remappedMin=0
      sourceComponent=RotationZ
      useQuaternion=true
      sourceBone {boneName=upperarm_l}
     }]
    targetBone {boneName=dyn_shoulderpad_l}
   }
   n75 = Flap Node - Bone: dyn_shoulderpad_l(ComponentPose=@n74, Alpha=0.4) {
    bUseClamp=true
    targetClampMax=1
    targetClampMin=-1
    targetComponent=RotationY
    alphaBoolBlend {blendOption=Linear}
    sourceDrives [{
      bUseRange=false
      drivingCurve=null
      multiplier=-0.8
      rangeMax=1
      rangeMin=-1
      remappedMax=1
      remappedMin=0
      sourceComponent=RotationY
      useQuaternion=true
      sourceBone {boneName=upperarm_l}
     }]
    targetBone {boneName=dyn_shoulderpad_l}
   }
   n76 = Flap Node - Bone: dyn_shoulderpad_r!(ComponentPose:Component Space Pose Link Structure=@n75, Alpha:float=0.4) -> (Pose:Component Space Pose Link Structure) {
    is=title FortAnimGraphNode_Flap
    bUseClamp=true
    targetClampMax=1
    targetClampMin=-1
    targetComponent=RotationZ
    alphaBoolBlend {blendOption=Linear}
    sourceDrives [{
      bUseRange=false
      drivingCurve=null
      multiplier=-0.8
      rangeMax=1
      rangeMin=-1
      remappedMax=1
      remappedMin=0
      sourceComponent=RotationZ
      useQuaternion=true
      sourceBone {boneName=upperarm_r}
     }]
    targetBone {boneName=dyn_shoulderpad_r}
   }
   n77 = Flap Node - Bone: dyn_shoulderpad_r(ComponentPose=@n76, Alpha=0.4) {
    bUseClamp=true
    targetClampMax=1
    targetClampMin=-1
    targetComponent=RotationY
    alphaBoolBlend {blendOption=Linear}
    sourceDrives [{
      bUseRange=false
      drivingCurve=null
      multiplier=-0.8
      rangeMax=1
      rangeMin=-1
      remappedMax=1
      remappedMin=0
      sourceComponent=RotationY
      useQuaternion=true
      sourceBone {boneName=upperarm_r}
     }]
    targetBone {boneName=dyn_shoulderpad_r}
   }
   Save cached pose 'shoulderpads_Skydiving'(Pose=Component To Local(ComponentPose=@n77)) {cacheName=shoulderpads_Skydiving}
   n78 = Get arm_up_r() -> (arm_up_r:double)
   n79 = "Transform (Modify) Bone - Bone: dyn_shoulderPad_fr_r"(
    ComponentPose=Local To Component(LocalPose=Use cached pose 'copypose_cache'())
    Alpha=@n78
    Translation="0.000,1.426,0.990"
    Rotation="-14.397,-34.323,-23.418"
    Scale="1,1,1"
   ) {
    rotationMode=BMM_Additive
    translationMode=BMM_Additive
    alphaBoolBlend {blendOption=Linear}
    boneToModify {boneName=dyn_shoulderPad_fr_r}
    rotation=<rot -14.397 -34.323 -23.418>
    translation {y=1.426, z=0.9901}
   }
   n80 = "Transform (Modify) Bone - Bone: dyn_shoulderpad_r"(ComponentPose=@n79, Alpha=@n78, Translation="0,0,0", Rotation="-7,0.000,0.000", Scale="1,1,1") {
    rotationMode=BMM_Additive
    translationMode=BMM_Additive
    alphaBoolBlend {blendOption=Linear}
    boneToModify {boneName=dyn_shoulderpad_r}
    rotation {pitch=-7}
   }
   n81 = Get arm_up_l() -> (arm_up_l:double)
   n82 = "Transform (Modify) Bone - Bone: dyn_shoulderPad_fr_l"(ComponentPose=@n80, Alpha=@n81, Translation="-1.724,1.425,0.000", Rotation="13.780,18.529,-32.404", Scale="1,1,1") {
    rotationMode=BMM_Additive
    translationMode=BMM_Additive
    alphaBoolBlend {blendOption=Linear}
    boneToModify {boneName=dyn_shoulderPad_fr_l}
    rotation=<rot 13.78 18.5292 -32.4042>
    translation=<-1.724 1.425>
   }
   n83 = "Transform (Modify) Bone - Bone: dyn_shoulderpad_l"(ComponentPose=@n82, Alpha=@n81, Translation="0,0,0", Rotation="7.223,0.000,0.000", Scale="1,1,1") {
    rotationMode=BMM_Additive
    translationMode=BMM_Additive
    alphaBoolBlend {blendOption=Linear}
    boneToModify {boneName=dyn_shoulderpad_l}
    rotation {pitch=7.223}
   }
   n84 = Get celebration_hack_alpha() -> (celebration_hack_alpha:double)
   n85 = "Transform (Modify) Bone - Bone: dyn_shoulderpad_r"(ComponentPose=@n83, Alpha=@n84, Translation="0,0,0", Rotation="-10,0.000,0.000", Scale="1,1,1") {
    rotationMode=BMM_Additive
    translationMode=BMM_Additive
    alphaBoolBlend {blendOption=Linear}
    boneToModify {boneName=dyn_shoulderpad_r}
    rotation {pitch=-10}
   }
   n86 = "Transform (Modify) Bone - Bone: dyn_shoulderpad_l"(ComponentPose=@n85, Alpha=@n84, Translation="0,0,0", Rotation="10,0.000,0.000", Scale="1,1,1") {
    rotationMode=BMM_Additive
    translationMode=BMM_Additive
    alphaBoolBlend {blendOption=Linear}
    boneToModify {boneName=dyn_shoulderpad_l}
    rotation {pitch=10}
   }
   n87 = Look At - Bone: dyn_shoulderPad_fr_r(ComponentPose=@n86, Alpha=Get arm_up_r()) {
    alphaBoolBlend {blendOption=Linear}
    boneToModify {boneName=dyn_shoulderPad_fr_r}
    lookAtLocation=<0 4>
    lookAtTarget {bUseSocket=true, boneReference {boneName=upperarm_twist_02_r}, socketReference {socketName=dyn_chest_aim_point_r}}
    lookAt_Axis {axis=<-1 0>}
   }
   n88 = Look At - Bone: dyn_shoulderPad_fr_l(ComponentPose=@n87, Alpha=Get arm_up_l()) {
    alphaBoolBlend {blendOption=Linear}
    boneToModify {boneName=dyn_shoulderPad_fr_l}
    lookAtLocation=<0 -4>
    lookAtTarget {bUseSocket=true, boneReference {boneName=upperarm_twist_02_l}, socketReference {socketName=dyn_chest_aim_point_l}}
    lookAt_Axis {axis=<1 0>}
   }
   Save cached pose 'shoulderpads_Grounded'(Pose=Component To Local(ComponentPose=@n88)) {cacheName=shoulderpads_Grounded}
   Save cached pose 'post_shoulderpads_cache'(Pose=Blend Poses by bool(
     BlendPose_0=Use cached pose 'shoulderpads_Skydiving'()
     BlendPose_1=Use cached pose 'shoulderpads_Grounded'()
     bActiveValue=Get bIsSkyDiving()
    ) {blendType=Linear}) {cacheName=post_shoulderpads_cache}
   n89 = "Constraint - dyn_skirt_front_01_l (P)"(ComponentPose=Local To Component(LocalPose=Use cached pose 'post_shoulderpads_cache'()), Alpha=1, ConstraintWeights_0=1) {
    alphaBoolBlend {blendOption=Linear}
    boneToModify {boneName=dyn_skirt_front_01_l}
    constraintSetup [{targetBone {boneName=thigh_l}}]
    constraintWeights [1]
   }
   n90 = "Constraint - dyn_skirt_front_01_r (P)"(ComponentPose=@n89, Alpha=1, ConstraintWeights_0=1) {
    alphaBoolBlend {blendOption=Linear}
    boneToModify {boneName=dyn_skirt_front_01_r}
    constraintSetup [{targetBone {boneName=thigh_r}}]
    constraintWeights [1]
   }
   n91 = RigidBody(ComponentPose=@n90, ExternalForce=Get skydiving_force() -> (skydiving_force:Vector)) {
    bClampLinearTranslationLimitToRefPose=true
    bOverrideWorldGravity=true
    overridePhysicsAsset=&/M_MED_Pirate_T3_Skirt_Physics
    simulationSpace=BaseBoneSpace
    alphaBoolBlend {blendOption=Linear}
    baseBoneRef {boneName=root}
    componentAppliedLinearAccClamp=<2500 2500 2500>
    overrideWorldGravity {z=-980}
   }
   Save cached pose 'rigidbody_cache'(Pose=Component To Local(ComponentPose=@n91)) {cacheName=rigidbody_cache}
   n92 = "Constraint - dyn_skirt_front_01_r (P)"(ComponentPose=Local To Component(LocalPose=Use cached pose 'post_shoulderpads_cache'()), Alpha=1, ConstraintWeights_0=1) {
    alphaBoolBlend {blendOption=Linear}
    boneToModify {boneName=dyn_skirt_front_01_r}
    constraintSetup [{targetBone {boneName=thigh_r}}]
    constraintWeights [1]
   }
   n93 = "Constraint - dyn_skirt_side_01_r (P)"(ComponentPose=@n92, Alpha=1, ConstraintWeights_0=1) {
    alphaBoolBlend {blendOption=Linear}
    boneToModify {boneName=dyn_skirt_side_01_r}
    constraintSetup [{targetBone {boneName=thigh_r}}]
    constraintWeights [1]
   }
   n94 = "Constraint - dyn_skirt_mid_01_r (P)"(ComponentPose=@n93, Alpha=1, ConstraintWeights_0=1) {
    alphaBoolBlend {blendOption=Linear}
    boneToModify {boneName=dyn_skirt_mid_01_r}
    constraintSetup [{targetBone {boneName=thigh_r}}]
    constraintWeights [1]
   }
   n95 = "Constraint - dyn_skirt_back_02_r (P)"(ComponentPose=@n94, Alpha=1, ConstraintWeights_0=1) {
    alphaBoolBlend {blendOption=Linear}
    boneToModify {boneName=dyn_skirt_back_02_r}
    constraintSetup [{targetBone {boneName=thigh_r}}]
    constraintWeights [1]
   }
   n96 = "Constraint - dyn_skirt_front_01_l (P)"(ComponentPose=@n95, Alpha=1, ConstraintWeights_0=1) {
    alphaBoolBlend {blendOption=Linear}
    boneToModify {boneName=dyn_skirt_front_01_l}
    constraintSetup [{targetBone {boneName=thigh_l}}]
    constraintWeights [1]
   }
   n97 = "Constraint - dyn_skirt_side_01_l (P)"(ComponentPose=@n96, Alpha=1, ConstraintWeights_0=1) {
    alphaBoolBlend {blendOption=Linear}
    boneToModify {boneName=dyn_skirt_side_01_l}
    constraintSetup [{targetBone {boneName=thigh_l}}]
    constraintWeights [1]
   }
   n98 = "Constraint - dyn_skirt_mid_01_l (P)"(ComponentPose=@n97, Alpha=1, ConstraintWeights_0=1) {
    alphaBoolBlend {blendOption=Linear}
    boneToModify {boneName=dyn_skirt_mid_01_l}
    constraintSetup [{targetBone {boneName=thigh_l}}]
    constraintWeights [1]
   }
   n99 = "Constraint - dyn_skirt_back_02_l (P)"(ComponentPose=@n98, Alpha=1, ConstraintWeights_0=1) {
    alphaBoolBlend {blendOption=Linear}
    boneToModify {boneName=dyn_skirt_back_02_l}
    constraintSetup [{targetBone {boneName=thigh_l}}]
    constraintWeights [1]
   }
   n100 = Get calf_bend_alpha_r() -> (calf_bend_alpha_r:double)
   n101 = "Transform (Modify) Bone - Bone: dyn_skirt_back_02_r"(ComponentPose=@n99, Alpha=@n100, Translation="0,0,0", Rotation="0,-25,0", Scale="1,1,1") {
    rotationMode=BMM_Additive
    rotationSpace=BCS_ParentBoneSpace
    alphaBoolBlend {blendOption=Linear}
    boneToModify {boneName=dyn_skirt_back_02_r}
    rotation {yaw=-25}
   }
   n102 = "Transform (Modify) Bone - Bone: dyn_skirt_back_03_r"(ComponentPose=@n101, Alpha=@n100, Translation="0,0,0", Rotation="0,-25,0", Scale="1,1,1") {
    rotationMode=BMM_Additive
    rotationSpace=BCS_ParentBoneSpace
    alphaBoolBlend {blendOption=Linear}
    boneToModify {boneName=dyn_skirt_back_03_r}
    rotation {yaw=-25}
   }
   n103 = "Transform (Modify) Bone - Bone: dyn_skirt_back_04_r"(ComponentPose=@n102, Alpha=@n100, Translation="0,0,0", Rotation="0,-25,0", Scale="1,1,1") {
    rotationMode=BMM_Additive
    rotationSpace=BCS_ParentBoneSpace
    alphaBoolBlend {blendOption=Linear}
    boneToModify {boneName=dyn_skirt_back_04_r}
    rotation {yaw=-25}
   }
   n104 = "Transform (Modify) Bone - Bone: dyn_skirt_mid_02_r"(ComponentPose=@n103, Alpha=@n100, Translation="0,0,0", Rotation="15,-10,0", Scale="1,1,1") {
    rotationMode=BMM_Additive
    rotationSpace=BCS_ParentBoneSpace
    alphaBoolBlend {blendOption=Linear}
    boneToModify {boneName=dyn_skirt_mid_02_r}
    rotation {pitch=15, yaw=-10}
   }
   n105 = "Transform (Modify) Bone - Bone: dyn_skirt_mid_03_r"(ComponentPose=@n104, Alpha=@n100, Translation="0,0,0", Rotation="15,-15,0", Scale="1,1,1") {
    rotationMode=BMM_Additive
    rotationSpace=BCS_ParentBoneSpace
    alphaBoolBlend {blendOption=Linear}
    boneToModify {boneName=dyn_skirt_mid_03_r}
    rotation {pitch=15, yaw=-15}
   }
   n106 = "Transform (Modify) Bone - Bone: dyn_skirt_mid_04_r"(ComponentPose=@n105, Alpha=@n100, Translation="0,0,0", Rotation="15,-20,0", Scale="1,1,1") {
    rotationMode=BMM_Additive
    rotationSpace=BCS_ParentBoneSpace
    alphaBoolBlend {blendOption=Linear}
    boneToModify {boneName=dyn_skirt_mid_04_r}
    rotation {pitch=15, yaw=-20}
   }
   n107 = Get calf_bend_alpha_l() -> (calf_bend_alpha_l:double)
   n108 = "Transform (Modify) Bone - Bone: dyn_skirt_back_02_l"(ComponentPose=@n106, Alpha=@n107, Translation="0,0,0", Rotation="0,-25,0", Scale="1,1,1") {
    rotationMode=BMM_Additive
    rotationSpace=BCS_ParentBoneSpace
    alphaBoolBlend {blendOption=Linear}
    boneToModify {boneName=dyn_skirt_back_02_l}
    rotation {yaw=-25}
   }
   n109 = "Transform (Modify) Bone - Bone: dyn_skirt_back_03_l"(ComponentPose=@n108, Alpha=@n107, Translation="0,0,0", Rotation="0,-25,0", Scale="1,1,1") {
    rotationMode=BMM_Additive
    rotationSpace=BCS_ParentBoneSpace
    alphaBoolBlend {blendOption=Linear}
    boneToModify {boneName=dyn_skirt_back_03_l}
    rotation {yaw=-25}
   }
   n110 = "Transform (Modify) Bone - Bone: dyn_skirt_back_04_l"(ComponentPose=@n109, Alpha=@n107, Translation="0,0,0", Rotation="0,-25,0", Scale="1,1,1") {
    rotationMode=BMM_Additive
    rotationSpace=BCS_ParentBoneSpace
    alphaBoolBlend {blendOption=Linear}
    boneToModify {boneName=dyn_skirt_back_04_l}
    rotation {yaw=-25}
   }
   n111 = "Transform (Modify) Bone - Bone: dyn_skirt_mid_02_l"(ComponentPose=@n110, Alpha=@n107, Translation="0,0,0", Rotation="-25,-10,0", Scale="1,1,1") {
    rotationMode=BMM_Additive
    rotationSpace=BCS_ParentBoneSpace
    alphaBoolBlend {blendOption=Linear}
    boneToModify {boneName=dyn_skirt_mid_02_l}
    rotation {pitch=-25, yaw=-10}
   }
   n112 = "Transform (Modify) Bone - Bone: dyn_skirt_mid_03_l"(ComponentPose=@n111, Alpha=@n107, Translation="0,0,0", Rotation="-25,-15,0", Scale="1,1,1") {
    rotationMode=BMM_Additive
    rotationSpace=BCS_ParentBoneSpace
    alphaBoolBlend {blendOption=Linear}
    boneToModify {boneName=dyn_skirt_mid_03_l}
    rotation {pitch=-25, yaw=-15}
   }
   n113 = "Transform (Modify) Bone - Bone: dyn_skirt_mid_04_l"(ComponentPose=@n112, Alpha=@n107, Translation="0,0,0", Rotation="-25,-20,0", Scale="1,1,1") {
    rotationMode=BMM_Additive
    rotationSpace=BCS_ParentBoneSpace
    alphaBoolBlend {blendOption=Linear}
    boneToModify {boneName=dyn_skirt_mid_04_l}
    rotation {pitch=-25, yaw=-20}
   }
   Save cached pose 'no_rigidbody_cache'(Pose=Component To Local(ComponentPose=@n113)) {cacheName=no_rigidbody_cache}
   vector * float(A=Get skydiving_animdyn_force(), B=0.45)
  ]}
 EventGraph {nodes [
   Event BlueprintUpdateAnimation()
   > n1 = Sequence!(execute) -> (then_0, then_1, then_2:Exec, then_3:Exec, then_4:Exec, then_5:Exec)
   n2 = GetAnimInstance(self=GetOwnerSkeletalMeshComponent())
   > Is Valid(InputObject=@n2)
   n3 = Montage_IsActive!(self:Anim Instance Object Reference=@n2, Montage:Anim Montage Object Reference=/Game/Animation/Game/MainPlayer/Emotes/Emote_Celebration_M) -> (ReturnValue:bool)
   > Set celebrate_hack(celebrate_hack:double=SelectFloat(A=0.2, B=1, bPickA=@n3)) -> (Output_Get:double)
   > Set celebration_hack_alpha(celebration_hack_alpha:double=Get Distance Between Two Sockets!(
     Component:Skeletal Mesh Component Object Reference=GetPartSkeletalMeshComponent()
     SocketOrBoneNameA:Name=lowerarm_r
     SocketSpaceA:ERelativeTransformSpace Enum=RTS_World
     SocketOrBoneNameB:Name=arm_straight_up_r
     SocketSpaceB:ERelativeTransformSpace Enum=RTS_World
     bRemapRange:bool=true
     InRangeMin:float=0
     InRangeMax:float=30
     OutRangeMin:float=1
     OutRangeMax:float=0.5
    ) -> (ReturnValue:float) {is=title K2Node_CallFunction}) -> (Output_Get:double)
   Branch(execute=@n1.then_1, Condition=Get bClothEnabled())
   > Set ClothBlendWeight!(
    execute
    ClothBlendWeight:float=Get skydiving_cloth_distance_scale() -> (skydiving_cloth_distance_scale:double)
    self:Skeletal Mesh Component Object Reference=GetPartSkeletalMeshComponent()
   ) -> (then, Output_Get:float)
   n4 = Get clothing_sim_interactor() -> (clothing_sim_interactor:Clothing Simulation Interactor Nv Object Reference)
   > SetAnimDriveSpringStiffness!(execute, self:Clothing Simulation Interactor Object Reference=@n4, InStiffness:float=1) -> (then)
   > SetAnimDriveDamperStiffness!(execute, self:Clothing Simulation Interactor Nv Object Reference=@n4, InStiffness:float=100) -> (then)
   Branch(execute=@n1.then_5, Condition=Get bIsInFrontEnd())
   > n5 = Cloth Controller Function DEPRECATED!(
    execute
    CharacterPartAnimInstance:Custom Character Part Anim Instance Object Reference=Self-Reference()
    ClothingAssetName:String=M_MED_Pirate_T4_Body_Clothing_1
    Cloth Paramaters:Cloth Params Struct Structure=Get Cloth Paramaters_0() -> (Cloth Paramaters_0:Cloth Params Struct Structure)
    Cloth Core Settings:Cloth Core Setting Struct Structure=Get Cloth Core Settings_0() -> (Cloth Core Settings_0:Cloth Core Setting Struct Structure)
    ClothUpAxis:Vector="1,0,0"
    OnGround_Moving_SpeedLimit:double=2
    OnGround_Moving_AngularVel_Max:double=0.05
    OnGround_Moving_FictitiousVel_Max:double=0.6
    LinVelScale_FwdBwd_Clamp:double=0
    LinVelScale_UpDwn_Clamp:double=0.7
    DeltaTime:double=Get DeltaTimeLocal() -> (DeltaTimeLocal:double)
    GameTimeInSeconds:double=GetGameTimeInSeconds()
    WindGustsParams:Wind Gust Params Struct Structure=Get Wind Gust Params() -> (Wind Gust Params:Wind Gust Params Struct Structure)
    WindGustLocalSettings:Wind Gust Local Setting Struct Structure=Get Gust Local Settings() -> (Gust Local Settings:Wind Gust Local Setting Struct Structure)
   ) -> (
    then
    CurrentStateOut:String
    Cloth Settings Out:Cloth Setting Struct Structure
    Cloth Core Settings Out:Cloth Core Setting Struct Structure
    WindGustSettings Out:Wind Gust Setting Struct Structure
    WindGustLocalSettings Out:Wind Gust Local Setting Struct Structure
    GustStrength - WindVelocityOut:Vector
    LinearVelVector-DELETE LATER:Vector
   )
   > Set Gust Local Settings(Gust Local Settings:Wind Gust Local Setting Struct Structure=@n5.WindGustLocalSettings Out) -> (Output_Get:Wind Gust Local Setting Struct Structure)
   n6 = BreakRotator(InRot=Get PawnRotationRate() -> (PawnRotationRate:Rotator))
   > Branch(Condition="float > float"(A="Absolute (Float)"(A=@n6.Yaw), B=3) {is=title K2Node_CallFunction})
   > SetClothingInteractorIfNotSet()
   > SetVelocityScale!(
    execute
    self:Chaos Clothing Interactor Object Reference=Get ChaosClothingInteractor() -> (ChaosClothingInteractor:Chaos Clothing Interactor Object Reference)
    LinearVelocityScale:Vector="0.05,0.05,0.05"
    AngularVelocityScale:float=0
    FictitiousAngularScale:float=0
   ) -> (then)
   n7 = GetOwnerSkeletalMeshComponent()
   n8 = BreakRotator(InRot=BreakTransform(InTransform=GetDeltaTransformFromRefPose(self=@n7, BoneName=upperarm_r, BaseName=spine_05)).Rotation)
   n9 = "Absolute (Float)"(A=@n8.Yaw)
   Set arm_up_r(
    execute=@n1.then_2
    arm_up_r:double=float * float(A=MapRangeClamped(Value=@n9, InRangeA=40, InRangeB=138), B=Get celebrate_hack() -> (celebrate_hack:double))
   ) -> (Output_Get:double)
   > Set arm_up_more_r(arm_up_more_r:double=MapRangeClamped(Value=@n9, InRangeA=85, InRangeB=125)) -> (Output_Get:double)
   > Set arm_back_r(arm_back_r:double=MapRangeClamped(Value=@n8.Yaw, InRangeA=35, InRangeB=-20)) -> (Output_Get:double)
   > Set arm_side_lift_r(arm_side_lift_r:double=MapRangeClamped(Value="Absolute (Float)"(A=@n8.Pitch), InRangeA=0, InRangeB=60)) -> (Output_Get:double)
   n10 = GetOwnerSkeletalMeshComponent()
   n11 = BreakRotator(InRot=BreakTransform(InTransform=GetDeltaTransformFromRefPose(self=@n10, BoneName=upperarm_r, BaseName=spine_05)).Rotation)
   n12 = "Absolute (Float)"(A=@n11.Yaw)
   > Set arm_up_l(arm_up_l:double=float * float(A=MapRangeClamped(Value=@n12, InRangeA=40, InRangeB=138), B=Get celebrate_hack())) -> (Output_Get:double)
   > Set arm_up_more_l(arm_up_more_l:double=MapRangeClamped(Value=@n12, InRangeA=85, InRangeB=125)) -> (Output_Get:double)
   > Set arm_back_l(arm_back_l:double=MapRangeClamped(Value=@n11.Yaw, InRangeA=35, InRangeB=-20)) -> (Output_Get:double)
   > Set arm_side_lift_l(arm_side_lift_l:double=MapRangeClamped(Value="Absolute (Float)"(A=@n11.Pitch), InRangeA=0, InRangeB=60)) -> (Output_Get:double)
   n13 = Break Fort Anim Input Skydiving External Force(FortAnimInput_SkydivingExternalForce=Get SkydivingExternalForceInput()) -> (
    bUseSkydivingVectorForce=~
    SkydivingVectorForce=~
    HeadToPelvisDirection=~
    FloatingMultiplier=~
    FloatingAdditive=~
    DivingMultiplier=~
    DivingAdditive=~
    ParachutingMultiplier=~
    ParachutingAdditive=~
   )
   Set skydiving_force(
    execute=@n1.then_3
    skydiving_force:Vector=vector * vector(A:Vector=@n13.SkydivingVectorForce, B:Vector="0.3,0.3,0.1") -> (ReturnValue:Vector) {is=title K2Node_CommutativeAssociativeBinaryOperator}
   ) -> (Output_Get:Vector)
   > Set tassle_skydiving_force(tassle_skydiving_force:Rotator="40,20, 0") -> (Output_Get:Rotator)
   > Set skydiving_animdyn_force(skydiving_animdyn_force:Vector=vector * float(A=@n13.SkydivingVectorForce, B=6)) -> (Output_Get:Vector)
   n14 = GetOwnerSkeletalMeshComponent()
   n15 = BreakTransform(InTransform=GetDeltaTransformFromRefPose(self=@n14, BoneName=calf_r, BaseName=None))
   Set calf_bend_alpha_r(execute=@n1.then_4, calf_bend_alpha_r:double=MapRangeClamped(Value=BreakRotator(InRot=@n15.Rotation).Yaw, InRangeA=95, InRangeB=115)) -> (Output_Get:double)
   n16 = BreakTransform(InTransform=GetDeltaTransformFromRefPose(self=@n14, BoneName=calf_l, BaseName=None))
   > Set calf_bend_alpha_l(calf_bend_alpha_l:double=MapRangeClamped(Value=BreakRotator(InRot=@n16.Rotation).Yaw, InRangeA=80, InRangeB=115)) -> (Output_Get:double)
   Event OnBeginSkydiving()
   > Set skydiving_alpha(skydiving_alpha:double=0) -> (Output_Get:double)
   > Set skydiving_cloth_distance_scale(skydiving_cloth_distance_scale:double=1) -> (Output_Get:double)
   > Set tassle_damping(tassle_damping:double=0.98) -> (Output_Get:double)
   Event OnEndSkydiving()
   > Set skydiving_alpha(skydiving_alpha=1)
   > Set skydiving_cloth_distance_scale(skydiving_cloth_distance_scale=0.5)
   > Set tassle_damping(tassle_damping=1)
   Event BlueprintInitializeAnimation()
   > n17 = Cast To ClothingSimulationInteractorNv!(
    execute
    Object=GetClothingSimulationInteractor!(self:Skeletal Mesh Component Object Reference=GetPartSkeletalMeshComponent(), ClothingAssetName:Name=None) -> (ReturnValue:Clothing Simulation Interactor Object Reference)
   ) -> (then, CastFailed, AsClothing Simulation Interactor Nv:Clothing Simulation Interactor Nv Object Reference) {is=title K2Node_DynamicCast}
   > Set clothing_sim_interactor(clothing_sim_interactor:Clothing Simulation Interactor Nv Object Reference=@n17) -> (Output_Get:Clothing Simulation Interactor Nv Object Reference)
   Reroute Node()
   n18 = GetOwnerSkeletalMeshComponent()
   n19 = BreakTransform(InTransform=GetDeltaTransformFromRefPose(self=@n18, BoneName=upperarm_r, BaseName=clavicle_r))
   > n20 = Set arm up r(arm up r:double=MapRangeClamped(Value=BreakRotator(InRot=@n19.Rotation).Yaw, InRangeA=30, InRangeB=120, OutRangeB=0.7)) -> (Output_Get:double)
   > Set arm_up_corrective_r(arm_up_corrective_r:double=float * float(A=@n20, B=2)) -> (Output_Get:double)
   n21 = BreakTransform(InTransform=GetDeltaTransformFromRefPose(self=@n18, BoneName=upperarm_l, BaseName=clavicle_l))
   > n22 = Set arm up l(arm up l:double=MapRangeClamped(Value=BreakRotator(InRot=@n21.Rotation).Yaw, InRangeA=30, InRangeB=120, OutRangeB=0.7)) -> (Output_Get:double)
   > Set arm_up_corrective_l(arm_up_corrective_l:double=float * float(A=@n22, B=2)) -> (Output_Get:double)
   Set arm_up_r(arm_up_r=float * float(A=0, B=Get arm_up_r()))
   > Set arm_up_l(arm_up_l=float * float(A=0, B=Get arm_up_l()))
   TryGetPawnOwner()
   SelectFloat(A=1, B=0, bPickA=@n3)
   Rotation From X Vector!(InVec:Vector=vector + vector(A:Vector="0, 0, 0", B:Vector="0,40,20") -> (ReturnValue:Vector) {is=title K2Node_CommutativeAssociativeBinaryOperator}) -> (ReturnValue:Rotator)
   n23 = BreakVector(InVec="0, 0, 0")
   MakeVector(
    Y=MapRangeClamped(Value=@n23.Y, InRangeA=300, InRangeB=1500, OutRangeA=-45, OutRangeB=45)
    Z=MapRangeClamped(Value=@n23.Z, InRangeA=500, InRangeB=1200, OutRangeA=-45, OutRangeB=45)
   )
   vector * vector(A="0, 0, 0", B="0,0.01,0.01")
   Event OnBeginCrouching()
   Event OnEndCrouching()
  ]}
 SetClothingInteractorIfNotSet {nodes [SetClothingInteractorIfNotSet() {is=title K2Node_FunctionEntry}]}
]
variables [
 arm up l {type=number}
 arm up r {type=number}
 arm_back_l {type=number, default=0.6364}
 arm_back_r {type=number, default=0.6364}
 arm_side_lift_l {type=number}
 arm_side_lift_r {type=number}
 arm_up_corrective_l {type=number}
 arm_up_corrective_r {type=number}
 arm_up_l {type=number}
 arm_up_more_l {type=number}
 arm_up_more_r {type=number}
 arm_up_r {type=number}
 calf_bend_alpha_l {type=number}
 calf_bend_alpha_r {type=number}
 celebrate_hack {type=number}
 celebration_hack_alpha {type=number}
 ChaosClothingInteractor {type=ChaosClothingInteractor, default=null}
 Cloth Core Settings_0 {
  type=ClothCoreSettingStruct
  default {
   collisionThickness=0
   dampingCoefficient=0.1
   frictionCoefficient=0.3
   localDampingCoefficient=0.1
   numberIterations=2
   numberSubsteps=1
   selfCollisionThickness=0
   animDriveDamping=<0 1>
   animDriveStiffness=<0 1>
   areaStiffness=<1 1>
   bendingStiffness=<1 1>
   edgeStiffness=<1 1>
   tetherStiffness=<1 1>
  }
 }
 Cloth Paramaters_0 {
  type=ClothParamsStruct
  default {
   dBNO {
    angularVelocityScale=0.6
    fictitiousAngularScale=0.1
    jointName=pelvis
    dragCoefficient=<0.2 1>
    gravityOverride=<-980 400 0>
    liftCoefficient=<0 0.5>
    linearVelocityScale=<0.3 0.3 0.1>
   }
   dBNO_Carried {
    angularVelocityScale=0.6
    fictitiousAngularScale=0.1
    jointName=pelvis
    dragCoefficient=<0.2 1>
    gravityOverride=<-980 -200 0>
    liftCoefficient=<0 0.5>
    linearVelocityScale=<0.3 0.3 0.1>
   }
   dBNO_Impulse {
    angularVelocityScale=0
    fictitiousAngularScale=0
    jointName=pelvis
    dragCoefficient=<0 0.2>
    gravityOverride=<-3000 -2000 0>
    liftCoefficient=<0 0.1>
    linearVelocityScale=<0.3 0.1 0.3>
   }
   dBNO_RollBwd {
    angularVelocityScale=0
    fictitiousAngularScale=0
    jointName=pelvis
    dragCoefficient=<0 0.2>
    gravityOverride=<-2000 0 0>
    liftCoefficient=<0 0.1>
    linearVelocityScale=<0.3 0.1 0.3>
   }
   dBNO_RollFwd {
    angularVelocityScale=0
    fictitiousAngularScale=0
    jointName=pelvis
    dragCoefficient=<0 0.2>
    gravityOverride=<-980 0 0>
    liftCoefficient=<0 0.1>
    linearVelocityScale=<0.3 0.1 0.3>
   }
   dBNO_Tumbling {
    angularVelocityScale=0
    fictitiousAngularScale=0
    jointName=pelvis
    dragCoefficient=<0 0.2>
    gravityOverride=<-2500 -300 0>
    liftCoefficient=<0 0.1>
    linearVelocityScale=<0.3 0.1 0.3>
   }
   emoteOrMelee {
    angularVelocityScale=0.2
    fictitiousAngularScale=0.1
    jointName=pelvis
    dragCoefficient=<0 0.3>
    gravityOverride=<-980 -500 0>
    liftCoefficient=<0 0.2>
    linearVelocityScale=<0.05 0.05 0.05>
   }
   front_End {
    angularVelocityScale=0.05
    fictitiousAngularScale=0.01
    jointName=root
    dragCoefficient=<0 2>
    gravityOverride=<0 -100 -980>
    liftCoefficient=<0 0.01>
    linearVelocityScale=<0.1 0.1 0.1>
   }
   onGround_Crouch_Moving {
    angularVelocityScale=0.1
    fictitiousAngularScale=0.3
    jointName=root
    dragCoefficient=<2 2>
    gravityOverride=<0 0 -980>
    liftCoefficient=<0.6 0.6>
    linearVelocityScale=<0.6 0.6 0.5>
   }
   onGround_Falling {
    angularVelocityScale=0.6
    fictitiousAngularScale=0.1
    jointName=root
    dragCoefficient=<0 0.2>
    gravityOverride=<0 -500 -980>
    liftCoefficient=<0 0.1>
    linearVelocityScale=<0.2 0.2 0.3>
   }
   onGround_Moving {
    angularVelocityScale=0.6
    fictitiousAngularScale=0.1
    jointName=root
    dragCoefficient=<0 0.3>
    gravityOverride=<0 -100 -980>
    liftCoefficient=<0 0.2>
    linearVelocityScale=<0.4 0.6 0.3>
   }
   onGround_Standing {
    angularVelocityScale=0.05
    fictitiousAngularScale=0.6
    jointName=root
    dragCoefficient=<0 2>
    gravityOverride=<0 -100 -980>
    liftCoefficient=<0 0.5>
    linearVelocityScale=<0.4 1 0.3>
   }
   rail_Grinding {
    angularVelocityScale=0.2
    fictitiousAngularScale=0
    jointName=root
    dragCoefficient=<0.1 0.2>
    gravityOverride=<50 -500 -400>
    liftCoefficient=<0 0.5>
    linearVelocityScale=<0.1 0.1 0.1>
   }
   riding_Moving {
    angularVelocityScale=0.05
    fictitiousAngularScale=1
    jointName=root
    dragCoefficient=<0.5 0.5>
    gravityOverride=<0 0 -980>
    liftCoefficient=<0.5 0.5>
    linearVelocityScale=<0.4 0.6 0.3>
   }
   riding_Moving_Targeting {
    angularVelocityScale=0.05
    fictitiousAngularScale=0.7
    jointName=root
    dragCoefficient=<0.5 0.5>
    gravityOverride=<0 0 -980>
    liftCoefficient=<0.5 0.5>
    linearVelocityScale=<0.2 0.6 0.2>
   }
   riding_Standing {
    angularVelocityScale=0.05
    fictitiousAngularScale=1
    jointName=root
    dragCoefficient=<0.5 0.5>
    gravityOverride=<0 -450 -980>
    liftCoefficient=<0.5 0.5>
    linearVelocityScale=<0.6 0.6 0.3>
   }
   skydive_DivingDown {
    angularVelocityScale=0.1
    fictitiousAngularScale=0
    jointName=pelvis
    dragCoefficient=<0 1.5>
    gravityOverride=<-450 -100 0>
    liftCoefficient=<0 0.3>
    linearVelocityScale=<0.1 0.1 0.1>
   }
   skydive_FloatingDown {
    angularVelocityScale=0.1
    fictitiousAngularScale=0
    jointName=root
    dragCoefficient=<0 2.5>
    gravityOverride=<0 0 980>
    liftCoefficient=<0 0.4>
    linearVelocityScale=<0.2 0.2 0.1>
   }
   skydive_GoingUp {
    angularVelocityScale=0.1
    fictitiousAngularScale=0
    jointName=root
    dragCoefficient=<0 0>
    gravityOverride=<0 -3000 3000>
    liftCoefficient=<0 0>
    linearVelocityScale=<0.1 0.1 0.05>
   }
   skydive_Parachuting {
    angularVelocityScale=0.1
    fictitiousAngularScale=0
    jointName=root
    dragCoefficient=<0 2.5>
    gravityOverride=<0 0 -980>
    liftCoefficient=<0 0.5>
    linearVelocityScale=<0.2 0.2 0.2>
   }
   skydive_Surfing {
    angularVelocityScale=0.6
    fictitiousAngularScale=0.1
    jointName=root
    dragCoefficient=<1.5 1.5>
    gravityOverride=<0 -600 -500>
    liftCoefficient=<1 1>
    linearVelocityScale=<0.1 0.1 0.3>
   }
   swimming {
    angularVelocityScale=0.6
    fictitiousAngularScale=0
    jointName=root
    dragCoefficient=<0 2>
    gravityOverride=<0 -600 400>
    liftCoefficient=<0 0.4>
    linearVelocityScale=<0.6 0.2 0.2>
   }
   tactical_Sprint {
    angularVelocityScale=0.6
    fictitiousAngularScale=0.1
    jointName=root
    dragCoefficient=<2 2>
    gravityOverride=<0 0 -980>
    liftCoefficient=<0.5 0.5>
    linearVelocityScale=<0.6 0.4 0.3>
   }
   ziplining {
    angularVelocityScale=0.6
    fictitiousAngularScale=0
    jointName=root
    dragCoefficient=<0 1.5>
    gravityOverride=<0 0 -980>
    liftCoefficient=<0 0.3>
    linearVelocityScale=<0.4 0.4 0.5>
   }
  }
 }
 clothing_sim_interactor {type=ClothingSimulationInteractorNv, default=null}
 DeltaTimeLocal {type=number}
 Gust Local Settings {type=WindGustLocalSettingStruct, default {elapsedTime=0, finalGustStrengthLocal=0, gustDuration=0, gustStrength=0, timeBetweenGusts=0}}
 skydiving_alpha {type=number, default=1}
 skydiving_animdyn_force {type=Vector}
 skydiving_cloth_distance_scale {type=number, default=0.5}
 skydiving_force {type=Vector}
 tassle_damping {type=number, default=1}
 tassle_skydiving_force {type=Rotator, default=<rot 40 20 0>}
 Wind Gust Params {
  type=WindGustParamsStruct
  default {
   dBNO {
    maxDownInterpSpeed=5
    maxGustDuration=2
    maxGustStrength=7
    maxTimeBetweenGusts=7
    maxUpInterpSpeed=5
    minDownInterpSpeed=2
    minGustDuration=0.5
    minGustStrength=3
    minTimeBetweenGusts=0
    minUpInterpSpeed=2
    windTurbulenceScalarA=0.1
    windTurbulenceScalarB=1
    windAmplitude=<0 0 0>
    windFrequency=<0 0 0>
    windOffset=<0 0 0>
   }
   dBNO_Carried {
    maxDownInterpSpeed=5
    maxGustDuration=2
    maxGustStrength=7
    maxTimeBetweenGusts=7
    maxUpInterpSpeed=5
    minDownInterpSpeed=2
    minGustDuration=0.5
    minGustStrength=3
    minTimeBetweenGusts=0
    minUpInterpSpeed=2
    windTurbulenceScalarA=0.1
    windTurbulenceScalarB=1
    windAmplitude=<0 0 0>
    windFrequency=<0 0 0>
    windOffset=<0 0 0>
   }
   dBNO_Impulse {
    maxDownInterpSpeed=5
    maxGustDuration=2
    maxGustStrength=7
    maxTimeBetweenGusts=7
    maxUpInterpSpeed=5
    minDownInterpSpeed=2
    minGustDuration=0.5
    minGustStrength=3
    minTimeBetweenGusts=0
    minUpInterpSpeed=2
    windTurbulenceScalarA=0.1
    windTurbulenceScalarB=1
    windAmplitude=<0 0 0>
    windFrequency=<0 0 0>
    windOffset=<0 0 0>
   }
   dBNO_RollBwd {
    maxDownInterpSpeed=5
    maxGustDuration=2
    maxGustStrength=7
    maxTimeBetweenGusts=7
    maxUpInterpSpeed=5
    minDownInterpSpeed=2
    minGustDuration=0.5
    minGustStrength=3
    minTimeBetweenGusts=0
    minUpInterpSpeed=2
    windTurbulenceScalarA=0.1
    windTurbulenceScalarB=1
    windAmplitude=<0 0 0>
    windFrequency=<0 0 0>
    windOffset=<0 0 0>
   }
   dBNO_RollFwd {
    maxDownInterpSpeed=5
    maxGustDuration=2
    maxGustStrength=7
    maxTimeBetweenGusts=7
    maxUpInterpSpeed=5
    minDownInterpSpeed=2
    minGustDuration=0.5
    minGustStrength=3
    minTimeBetweenGusts=0
    minUpInterpSpeed=2
    windTurbulenceScalarA=0.1
    windTurbulenceScalarB=1
    windAmplitude=<0 0 0>
    windFrequency=<0 0 0>
    windOffset=<0 0 0>
   }
   dBNO_Tumbling {
    maxDownInterpSpeed=5
    maxGustDuration=2
    maxGustStrength=7
    maxTimeBetweenGusts=7
    maxUpInterpSpeed=5
    minDownInterpSpeed=2
    minGustDuration=0.5
    minGustStrength=3
    minTimeBetweenGusts=0
    minUpInterpSpeed=2
    windTurbulenceScalarA=0.1
    windTurbulenceScalarB=1
    windAmplitude=<0 0 0>
    windFrequency=<0 0 0>
    windOffset=<0 0 0>
   }
   emoteOrMelee {
    maxDownInterpSpeed=5
    maxGustDuration=2
    maxGustStrength=7
    maxTimeBetweenGusts=7
    maxUpInterpSpeed=5
    minDownInterpSpeed=2
    minGustDuration=0.5
    minGustStrength=3
    minTimeBetweenGusts=0
    minUpInterpSpeed=2
    windTurbulenceScalarA=0.1
    windTurbulenceScalarB=1
    windAmplitude=<30 25 50>
    windFrequency=<2 3 5>
    windOffset=<-300 -150 0>
   }
   front_End {
    maxDownInterpSpeed=5
    maxGustDuration=2
    maxGustStrength=4
    maxTimeBetweenGusts=7
    maxUpInterpSpeed=5
    minDownInterpSpeed=2
    minGustDuration=0.5
    minGustStrength=2
    minTimeBetweenGusts=0
    minUpInterpSpeed=2
    windTurbulenceScalarA=0.1
    windTurbulenceScalarB=1
    windAmplitude=<10 30 50>
    windFrequency=<10 3 4>
    windOffset=<0 -200 -250>
   }
   onGround_Crouch_Moving {
    maxDownInterpSpeed=5
    maxGustDuration=2
    maxGustStrength=7
    maxTimeBetweenGusts=7
    maxUpInterpSpeed=5
    minDownInterpSpeed=2
    minGustDuration=0.5
    minGustStrength=3
    minTimeBetweenGusts=0
    minUpInterpSpeed=2
    windTurbulenceScalarA=0.1
    windTurbulenceScalarB=1
    windAmplitude=<100 25 75>
    windFrequency=<1.5 0.5 7.5>
    windOffset=<0 -100 -175>
   }
   onGround_Falling {
    maxDownInterpSpeed=5
    maxGustDuration=2
    maxGustStrength=7
    maxTimeBetweenGusts=7
    maxUpInterpSpeed=5
    minDownInterpSpeed=2
    minGustDuration=0.5
    minGustStrength=3
    minTimeBetweenGusts=0
    minUpInterpSpeed=2
    windTurbulenceScalarA=0.1
    windTurbulenceScalarB=1
    windAmplitude=<0 0 0>
    windFrequency=<0 0 0>
    windOffset=<0 0 0>
   }
   onGround_Moving {
    maxDownInterpSpeed=5
    maxGustDuration=8
    maxGustStrength=7
    maxTimeBetweenGusts=8
    maxUpInterpSpeed=5
    minDownInterpSpeed=2
    minGustDuration=0.2
    minGustStrength=3
    minTimeBetweenGusts=1
    minUpInterpSpeed=2
    windTurbulenceScalarA=0.1
    windTurbulenceScalarB=0.8
    windAmplitude=<30 65 42>
    windFrequency=<5 6 8.6>
    windOffset=<30 -250 -300>
   }
   onGround_Standing {
    maxDownInterpSpeed=5
    maxGustDuration=8
    maxGustStrength=7
    maxTimeBetweenGusts=8
    maxUpInterpSpeed=5
    minDownInterpSpeed=2
    minGustDuration=0.2
    minGustStrength=2
    minTimeBetweenGusts=1
    minUpInterpSpeed=2
    windTurbulenceScalarA=0.1
    windTurbulenceScalarB=0.8
    windAmplitude=<75 24 28>
    windFrequency=<1.5 2.3 3.7>
    windOffset=<75 -100 -150>
   }
   rail_Grinding {
    maxDownInterpSpeed=5
    maxGustDuration=8
    maxGustStrength=7
    maxTimeBetweenGusts=8
    maxUpInterpSpeed=5
    minDownInterpSpeed=2
    minGustDuration=0.2
    minGustStrength=3
    minTimeBetweenGusts=1
    minUpInterpSpeed=2
    windTurbulenceScalarA=0.1
    windTurbulenceScalarB=0.8
    windAmplitude=<20 65 42>
    windFrequency=<5 6 8.6>
    windOffset=<50 -500 -400>
   }
   riding_Moving {
    maxDownInterpSpeed=5
    maxGustDuration=2
    maxGustStrength=7
    maxTimeBetweenGusts=7
    maxUpInterpSpeed=5
    minDownInterpSpeed=2
    minGustDuration=0.5
    minGustStrength=3
    minTimeBetweenGusts=0
    minUpInterpSpeed=2
    windTurbulenceScalarA=0.1
    windTurbulenceScalarB=1
    windAmplitude=<0 0 0>
    windFrequency=<0 0 0>
    windOffset=<0 0 0>
   }
   riding_Moving_Targeting {
    maxDownInterpSpeed=5
    maxGustDuration=2
    maxGustStrength=7
    maxTimeBetweenGusts=7
    maxUpInterpSpeed=5
    minDownInterpSpeed=2
    minGustDuration=0.5
    minGustStrength=3
    minTimeBetweenGusts=0
    minUpInterpSpeed=2
    windTurbulenceScalarA=0.1
    windTurbulenceScalarB=1
    windAmplitude=<0 0 0>
    windFrequency=<0 0 0>
    windOffset=<0 0 0>
   }
   riding_Standing {
    maxDownInterpSpeed=5
    maxGustDuration=2
    maxGustStrength=7
    maxTimeBetweenGusts=7
    maxUpInterpSpeed=5
    minDownInterpSpeed=2
    minGustDuration=0.5
    minGustStrength=3
    minTimeBetweenGusts=0
    minUpInterpSpeed=2
    windTurbulenceScalarA=0.1
    windTurbulenceScalarB=1
    windAmplitude=<0 0 0>
    windFrequency=<0 0 0>
    windOffset=<0 0 0>
   }
   skydive_DivingDown {
    maxDownInterpSpeed=5
    maxGustDuration=8
    maxGustStrength=7
    maxTimeBetweenGusts=8
    maxUpInterpSpeed=5
    minDownInterpSpeed=2
    minGustDuration=0.2
    minGustStrength=3
    minTimeBetweenGusts=1
    minUpInterpSpeed=2
    windTurbulenceScalarA=0.1
    windTurbulenceScalarB=4
    windAmplitude=<0 25 35>
    windFrequency=<0 4 10>
    windOffset=<-450 -100 100>
   }
   skydive_FloatingDown {
    maxDownInterpSpeed=5
    maxGustDuration=8
    maxGustStrength=1
    maxTimeBetweenGusts=8
    maxUpInterpSpeed=5
    minDownInterpSpeed=2
    minGustDuration=0.2
    minGustStrength=0
    minTimeBetweenGusts=1
    minUpInterpSpeed=2
    windTurbulenceScalarA=0.5
    windTurbulenceScalarB=1
    windAmplitude=<50 20 35>
    windFrequency=<8 2 3>
    windOffset=<0 -300 200>
   }
   skydive_GoingUp {
    maxDownInterpSpeed=5
    maxGustDuration=8
    maxGustStrength=7
    maxTimeBetweenGusts=8
    maxUpInterpSpeed=5
    minDownInterpSpeed=2
    minGustDuration=0.2
    minGustStrength=3
    minTimeBetweenGusts=1
    minUpInterpSpeed=2
    windTurbulenceScalarA=0.1
    windTurbulenceScalarB=0.8
    windAmplitude=<0 0 0>
    windFrequency=<0 0 0>
    windOffset=<0 -500 300>
   }
   skydive_Parachuting {
    maxDownInterpSpeed=5
    maxGustDuration=8
    maxGustStrength=1
    maxTimeBetweenGusts=8
    maxUpInterpSpeed=5
    minDownInterpSpeed=2
    minGustDuration=5
    minGustStrength=0
    minTimeBetweenGusts=5
    minUpInterpSpeed=2
    windTurbulenceScalarA=0.2
    windTurbulenceScalarB=1
    windAmplitude=<25 50 75>
    windFrequency=<8 6 2>
    windOffset=<-150 -400 -150>
   }
   skydive_Surfing {
    maxDownInterpSpeed=5
    maxGustDuration=2
    maxGustStrength=7
    maxTimeBetweenGusts=7
    maxUpInterpSpeed=5
    minDownInterpSpeed=2
    minGustDuration=0.5
    minGustStrength=3
    minTimeBetweenGusts=0
    minUpInterpSpeed=2
    windTurbulenceScalarA=0.1
    windTurbulenceScalarB=1
    windAmplitude=<0 0 0>
    windFrequency=<0 0 0>
    windOffset=<0 0 0>
   }
   swimming {
    maxDownInterpSpeed=5
    maxGustDuration=2
    maxGustStrength=7
    maxTimeBetweenGusts=7
    maxUpInterpSpeed=5
    minDownInterpSpeed=2
    minGustDuration=0.5
    minGustStrength=3
    minTimeBetweenGusts=0
    minUpInterpSpeed=2
    windTurbulenceScalarA=0.1
    windTurbulenceScalarB=1
    windAmplitude=<10 25 20>
    windFrequency=<1.5 0.5 7.5>
    windOffset=<0 -150 50>
   }
   tactical_Sprint {
    maxDownInterpSpeed=5
    maxGustDuration=2
    maxGustStrength=7
    maxTimeBetweenGusts=7
    maxUpInterpSpeed=5
    minDownInterpSpeed=2
    minGustDuration=0.5
    minGustStrength=3
    minTimeBetweenGusts=0
    minUpInterpSpeed=2
    windTurbulenceScalarA=0.1
    windTurbulenceScalarB=1
    windAmplitude=<0 0 0>
    windFrequency=<0 0 0>
    windOffset=<0 0 0>
   }
   ziplining {
    maxDownInterpSpeed=5
    maxGustDuration=2
    maxGustStrength=7
    maxTimeBetweenGusts=7
    maxUpInterpSpeed=5
    minDownInterpSpeed=2
    minGustDuration=0.5
    minGustStrength=3
    minTimeBetweenGusts=0
    minUpInterpSpeed=2
    windTurbulenceScalarA=0.1
    windTurbulenceScalarB=1
    windAmplitude=<0 0 0>
    windFrequency=<0 0 0>
    windOffset=<0 0 0>
   }
  }
 }
]
