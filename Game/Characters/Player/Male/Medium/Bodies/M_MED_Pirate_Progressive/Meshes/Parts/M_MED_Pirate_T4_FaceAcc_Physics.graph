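# Physics asset for the T4 face accessory of the medium male progressive pirate body, setting up the beard and its braids for physics simulation.
# Capsule bodies chain from the head through two beard bones, with mirrored A braids hanging from the first and B braids from the second.
# The beard root and braid master joints rotate freely, while the second beard bone and braid ends have narrow twist and uneven swing limits to restrict bending.
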PhysicsAsset /Game/Characters/Player/Male/Medium/Bodies/M_MED_Pirate_Progressive/Meshes/Parts/M_MED_Pirate_T4_FaceAcc_Physics
bodies {
 dyn_beard_01 [{shapeType=Capsule, length=5, radius=2, center=<2.5 0 0>, rotation=<rot 90 0 0>}]
 dyn_beard_02 [{shapeType=Capsule, length=5, radius=2, center=<2.5 0 0>, rotation=<rot 90 0 0>}]
 dyn_beard_braid_A_T3_01_l [{shapeType=Capsule, length=5, radius=3, center=<2 0 0>, rotation=<rot 90 0 0>}]
 dyn_beard_braid_A_T3_01_r [{shapeType=Capsule, length=5, radius=3, center=<-2 0 0>, rotation=<rot 90 0 0>}]
 dyn_beard_braid_A_T3_master_l [{shapeType=Capsule, length=5, radius=2, center=<2.5 0 0>, rotation=<rot 90 0 0>}]
 dyn_beard_braid_A_T3_master_r [{shapeType=Capsule, length=5, radius=2, center=<-2.5 0 0>, rotation=<rot 90 0 0>}]
 dyn_beard_braid_B_T3_01_l [{shapeType=Capsule, length=5, radius=3, center=<2.5 0 0>, rotation=<rot 90 0 0>}]
 dyn_beard_braid_B_T3_01_r [{shapeType=Capsule, length=5, radius=3, center=<-2.5 0 0>, rotation=<rot 90 0 0>}]
 dyn_beard_braid_B_T3_master_l [{shapeType=Capsule, length=5, radius=2, center=<2.5 0 0>, rotation=<rot 90 0 0>}]
 dyn_beard_braid_B_T3_master_r [{shapeType=Capsule, length=5, radius=2, center=<-2.5 0 0>, rotation=<rot 90 0 0>}]
 head [{shapeType=Capsule, length=1, radius=1}]
}
constraints [
 {bone1Name=dyn_beard_02, bone2Name=dyn_beard_01, swing1LimitDegrees=45, swing2LimitDegrees=0, twistLimitDegrees=1}
 {
  bone1Name=dyn_beard_braid_B_T3_master_l
  bone2Name=dyn_beard_02
  swing1LimitDegrees=45
  swing1Motion=Free
  swing2LimitDegrees=45
  swing2Motion=Free
  twistLimitDegrees=45
  twistMotion=Free
 }
 {
  bone1Name=dyn_beard_braid_B_T3_master_r
  bone2Name=dyn_beard_02
  swing1LimitDegrees=45
  swing1Motion=Free
  swing2LimitDegrees=45
  swing2Motion=Free
  twistLimitDegrees=45
  twistMotion=Free
 }
 {
  bone1Name=dyn_beard_braid_A_T3_master_l
  bone2Name=dyn_beard_01
  swing1LimitDegrees=45
  swing1Motion=Free
  swing2LimitDegrees=45
  swing2Motion=Free
  twistLimitDegrees=45
  twistMotion=Free
 }
 {
  bone1Name=dyn_beard_braid_A_T3_master_r
  bone2Name=dyn_beard_01
  swing1LimitDegrees=45
  swing1Motion=Free
  swing2LimitDegrees=45
  swing2Motion=Free
  twistLimitDegrees=45
  twistMotion=Free
 }
 {
  bone1Name=dyn_beard_braid_A_T3_01_r
  bone2Name=dyn_beard_braid_A_T3_master_r
  swing1LimitDegrees=10
  swing2LimitDegrees=20
  twistLimitDegrees=1
 }
 {
  bone1Name=dyn_beard_braid_A_T3_01_l
  bone2Name=dyn_beard_braid_A_T3_master_l
  swing1LimitDegrees=10
  swing2LimitDegrees=20
  twistLimitDegrees=1
 }
 {
  bone1Name=dyn_beard_braid_B_T3_01_r
  bone2Name=dyn_beard_braid_B_T3_master_r
  swing1LimitDegrees=1
  swing2LimitDegrees=30
  twistLimitDegrees=1
 }
 {
  bone1Name=dyn_beard_braid_B_T3_01_l
  bone2Name=dyn_beard_braid_B_T3_master_l
  swing1LimitDegrees=1
  swing2LimitDegrees=30
  twistLimitDegrees=1
 }
 {
  bone1Name=dyn_beard_01
  bone2Name=head
  swing1LimitDegrees=45
  swing1Motion=Free
  swing2LimitDegrees=45
  swing2Motion=Free
  twistLimitDegrees=45
  twistMotion=Free
 }
]
